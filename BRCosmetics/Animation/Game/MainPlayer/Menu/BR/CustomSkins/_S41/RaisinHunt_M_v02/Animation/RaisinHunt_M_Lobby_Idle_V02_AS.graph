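# Level Sequence for the RaisinHunt_M V02 skin's lobby idle animation in the battle royale menu.
# It spawns a FortPlayerMannequin at the origin, with Control Rig tracks for the body (MR_Biped_v02) and face (Fortnite_3L_OnFace_CtrlRig).
# The face rig carries dense keyed curves, including shared eye-aim motion with offsets on the right eye.
# The visible skeletal animation tracks and body rig section hold no animation data.
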LevelSequence /BRCosmetics/Animation/Game/MainPlayer/Menu/BR/CustomSkins/_S41/RaisinHunt_M_v02/Animation/RaisinHunt_M_Lobby_Idle_V02_AS
bindings [FortPlayerMannequin {
  id=#1
  possessed_class=FortPlayerMannequin
  tracks [
   {type=MovieSceneSpawnTrack, sections [{type=MovieSceneSpawnSection, channels [None {type=MovieSceneScriptingBoolChannel, default=true}]}]}
   {
    type=MovieScene3DTransformTrack
    sections [{
      type=MovieScene3DTransformSection
      channels [
       Location.X {type=MovieSceneScriptingDoubleChannel, default=0}
       Location.Y {type=MovieSceneScriptingDoubleChannel, default=0}
       Location.Z {type=MovieSceneScriptingDoubleChannel, default=0}
       Rotation.X {type=MovieSceneScriptingDoubleChannel, default=0}
       Rotation.Y {type=MovieSceneScriptingDoubleChannel, default=0}
       Rotation.Z {type=MovieSceneScriptingDoubleChannel, default=0}
       Scale.X {type=MovieSceneScriptingDoubleChannel, default=1}
       Scale.Y {type=MovieSceneScriptingDoubleChannel, default=1}
       Scale.Z {type=MovieSceneScriptingDoubleChannel, default=1}
      ]
     }]
   }
   {type=MovieSceneSkeletalAnimationTrack, sections []}
   {type=MovieSceneControlRigParameterTrack, display_name=MR_Biped_v02, sections [{type=MovieSceneControlRigParameterSection}]}
   {type=MovieSceneSkeletalAnimationTrack, sections []}
   {
    type=MovieSceneControlRigParameterTrack
    display_name=Fortnite_3L_OnFace_CtrlRig
    sections [
     {
      type=MovieSceneControlRigParameterSection
      channels [
       CTRL_C_eyesAim.Location.X {
        type=MovieSceneScriptingFloatChannel
        default=0
        thinned_from=575
        keys [[0,-7.9226],[4,-7.968],[14,-8.2198],[18,-8.2431],[21,-8.1832],[32,-7.7327],[35,-7.6547],[37,-7.6422],[38,-7.8303],[40,-8.6725],[41,-8.8628],[44,-6.4567],[45,-6.5268],[46,-6.6179],[48,-6.8447],[50,-7.1019],[51,-7.2462],[54,-7.7308],[55,-7.8628],[56,-7.9584],[61,-8.1795],[68,-8.3371],[76,-8.3808],[92,-8.1623],[99,-7.9593],[101,-7.9279],[102,-8.0189],[103,-9.075],[105,-9.1192],[151,-8.9016],[166,-8.5408],[170,-8.5286],[172,-8.567],[173,-8.6184],[174,-8.6913],[176,-8.8918],[178,-9.1481],[185,-10.2],[194,-11.2866],[201,-11.9705],[206,-12.3347],[211,-12.5621],[215,-12.6285],[217,-12.6149],[220,-12.4988],[227,-12.0852],[229,-12.0206],[233,-12.0332],[240,-12.267],[257,-13.1736],[263,-13.3402],[266,-13.3444],[268,-13.2626],[270,-13.1097],[273,-12.7749],[278,-12.0488],[279,-11.8703],[280,-11.6497],[281,-11.3954],[286,-9.9272],[287,-9.6579],[288,-9.4187],[289,-9.2191],[290,-9.0686],[291,-8.9768],[304,-8.2523],[310,-8.0157],[315,-7.9076],[350,-7.7756],[353,-7.8356],[364,-8.288],[367,-8.3642],[372,-8.3613],[381,-8.1745],[392,-7.7939],[394,-7.6908],[396,-7.537],[398,-7.3438],[406,-6.4167],[408,-6.2148],[410,-6.0464],[412,-5.9202],[413,-5.8758],[414,-5.8454],[415,-5.8302],[416,-5.8315],[417,-6.1692],[418,-6.9463],[419,-7.8385],[420,-8.5278],[421,-9.0067],[422,-9.4616],[423,-9.8713],[424,-10.2148],[425,-10.4719],[426,-10.6221],[432,-11.117],[435,-11.2891],[438,-11.394],[441,-11.4129],[447,-11.3406],[448,-11.3701],[449,-11.4567],[450,-11.5939],[451,-11.7755],[452,-11.9949],[453,-12.2457],[455,-12.8158],[460,-14.3495],[462,-14.8834],[463,-15.1115],[464,-15.3069],[465,-15.4647],[466,-15.5797],[467,-15.6469],[468,-15.6613],[470,-15.5733],[472,-15.3867],[479,-14.4345],[481,-14.2189],[483,-14.0939],[484,-14.0767],[485,-14.9475],[486,-15.0069],[487,-14.9671],[488,-14.8272],[489,-14.6165],[490,-14.3536],[493,-13.4431],[494,-13.1663],[495,-12.9382],[496,-12.781],[500,-12.3492],[503,-12.0971],[505,-11.9773],[507,-11.9102],[509,-11.9109],[511,-11.9941],[519,-12.6403],[521,-12.7467],[522,-12.7671],[523,-12.6534],[524,-12.3606],[525,-11.9128],[526,-11.3268],[527,-10.8869],[528,-10.7029],[529,-10.5702],[530,-10.4608],[531,-10.3281],[532,-10.1357],[533,-9.8877],[534,-9.5902],[537,-8.613],[538,-8.3174],[539,-8.0609],[540,-7.8579],[541,-7.7245],[542,-7.6771],[543,-7.6872],[545,-7.7523],[549,-7.947],[551,-8.0114],[553,-8.0211],[561,-7.9306],[576,-7.9306]]
       }
       CTRL_C_eyesAim.Location.Y {
        type=MovieSceneScriptingFloatChannel
        default=0
        thinned_from=575
        keys [[0,9.133],[22,9.1797],[37,9.077],[38,9.1339],[40,9.4071],[41,9.4739],[44,8.3404],[51,8.479],[101,8.4644],[102,8.486],[103,10.485],[104,10.4295],[105,10.2883],[106,10.124],[107,9.9906],[109,9.809],[114,9.597],[118,9.5063],[165,9.2572],[168,9.3086],[171,9.4361],[173,9.5704],[175,9.7503],[178,10.0933],[188,11.5024],[191,11.8568],[193,12.0435],[195,12.1767],[198,12.2676],[202,12.2533],[228,11.2168],[236,11.1422],[265,11.644],[268,11.6229],[271,11.5241],[278,11.0788],[280,10.8962],[286,10.2084],[288,10.0178],[290,9.8735],[305,9.2775],[314,9.0815],[356,9.0503],[369,9.1275],[405,8.6582],[409,8.6918],[416,8.9038],[417,8.9726],[418,9.1033],[420,9.4306],[423,9.7991],[424,9.9087],[425,9.9966],[432,10.2519],[436,10.333],[443,10.6562],[445,10.7117],[448,10.7134],[450,10.7574],[452,10.8511],[455,11.0724],[462,11.7694],[464,11.9241],[466,12.0069],[467,12.0116],[468,11.9867],[470,11.873],[479,11.0575],[481,10.928],[483,10.8552],[484,10.8453],[485,13.3672],[486,13.7789],[487,14.1424],[488,14.0465],[489,13.8724],[490,13.6399],[494,12.5116],[495,12.2656],[496,12.0634],[497,11.8982],[498,11.7583],[499,11.6494],[503,11.4347],[507,11.31],[510,11.2892],[516,11.4293],[517,11.4873],[518,11.5851],[521,11.9743],[522,12.0839],[523,12.1116],[524,12.0259],[525,11.1551],[526,9.9687],[527,9.4604],[528,9.3596],[529,9.3325],[530,9.3443],[532,9.4049],[533,9.4164],[534,9.3859],[538,9.1749],[541,9.0795],[576,9.1348]]
       }
       CTRL_C_eyesAim.Location.Z {
        type=MovieSceneScriptingFloatChannel
        default=0
        thinned_from=574
        keys [[0,1.7148],[16,1.7487],[36,1.6925],[37,1.6925],[38,1.6275],[40,1.3524],[41,1.2908],[45,10.2644],[50,10.3457],[51,10.267],[52,10.0349],[53,9.7097],[54,9.352],[55,9.0227],[56,8.7829],[59,8.2659],[61,7.9694],[62,7.8472],[63,7.7474],[64,7.6737],[65,7.6298],[82,7.2018],[91,7.1136],[92,7.1396],[93,7.2148],[94,7.3282],[95,7.4691],[98,7.9475],[99,8.0893],[100,8.2041],[101,8.2811],[102,8.3221],[103,-1.5446],[104,-1.3858],[105,-1.0127],[106,-0.5649],[107,-0.185],[108,0.1068],[109,0.3621],[110,0.5425],[111,0.6722],[112,0.7989],[113,0.9195],[114,1.0311],[115,1.1305],[116,1.2146],[117,1.2803],[118,1.3245],[123,1.4826],[126,1.5676],[128,1.6168],[130,1.6578],[132,1.6887],[148,1.8315],[153,1.8568],[160,1.8602],[164,1.8501],[165,1.83],[166,1.7779],[167,1.6961],[168,1.5867],[169,1.4522],[170,1.2948],[171,1.1169],[172,0.9221],[173,0.7141],[174,0.4946],[175,0.2658],[176,0.0296],[177,-0.2121],[178,-0.4572],[180,-0.95],[181,-1.1941],[182,-1.4341],[183,-1.6683],[184,-1.895],[185,-2.1127],[186,-2.3197],[187,-2.5147],[188,-2.6951],[189,-2.8585],[190,-3.003],[191,-3.127],[192,-3.2288],[193,-3.3066],[194,-3.3587],[195,-3.3836],[196,-3.3793],[197,-3.3503],[198,-3.3033],[199,-3.2398],[200,-3.1615],[201,-3.0699],[202,-2.9665],[203,-2.8528],[204,-2.7305],[205,-2.601],[207,-2.3267],[211,-1.7613],[212,-1.6257],[213,-1.4957],[214,-1.3729],[215,-1.2589],[216,-1.1554],[217,-1.0649],[218,-0.9897],[219,-0.9308],[220,-0.8894],[227,-0.6763],[228,-0.6444],[229,-0.6113],[233,-0.4709],[234,-0.4369],[235,-0.4038],[236,-0.372],[237,-0.3414],[238,-0.3125],[239,-0.2854],[240,-0.2602],[241,-0.2372],[242,-0.2166],[243,-0.1986],[244,-0.1834],[245,-0.1712],[246,-0.1611],[247,-0.152],[248,-0.144],[249,-0.1367],[250,-0.1302],[251,-0.1242],[252,-0.1188],[253,-0.1137],[257,-0.095],[258,-0.0901],[259,-0.085],[260,-0.0801],[261,-0.0762],[262,-0.0765],[263,-0.0834],[264,-0.0955],[265,-0.1116],[266,-0.1319],[267,-0.1562],[268,-0.1828],[269,-0.2099],[270,-0.2358],[271,-0.2586],[272,-0.2766],[273,-0.2891],[274,-0.2962],[275,-0.2977],[276,-0.2936],[277,-0.2837],[278,-0.268],[279,-0.2508],[281,-0.2196],[282,-0.2026],[283,-0.1828],[284,-0.159],[285,-0.1299],[286,-0.0946],[287,-0.0521],[288,-0.0016],[289,0.0578],[290,0.1268],[291,0.2063],[292,0.2919],[293,0.3787],[299,0.908],[301,1.0801],[303,1.2458],[305,1.4027],[306,1.4769],[307,1.548],[308,1.6155],[309,1.6792],[310,1.7388],[311,1.7939],[312,1.8441],[313,1.8893],[314,1.929],[315,1.9629],[316,1.9908],[317,2.0123],[326,2.107],[329,2.1273],[330,2.1273],[331,2.1351],[346,2.0919],[347,2.0919],[348,2.084],[349,2.084],[350,2.078],[355,2.0826],[364,2.1344],[367,2.1678],[371,2.231],[377,2.3611],[389,2.6569],[393,2.7353],[395,2.7592],[397,2.7699],[399,2.768],[401,2.7538],[403,2.7278],[404,2.6963],[405,2.6359],[406,2.5506],[407,2.444],[408,2.3198],[409,2.1819],[410,2.034],[413,1.5667],[414,1.4153],[415,1.2721],[416,1.1406],[417,1.0628],[418,1.0572],[420,1.1258],[422,1.1676],[423,1.1841],[424,1.1936],[425,1.193],[426,1.179],[427,1.1589],[428,1.1424],[429,1.13],[430,1.122],[431,1.1186],[432,1.1204],[433,1.1275],[435,1.1489],[436,1.1267],[437,1.053],[438,0.9388],[439,0.7951],[440,0.6332],[441,0.4647],[442,0.301],[443,0.1536],[444,0.0339],[445,-0.047],[446,-0.0776],[447,-0.0697],[448,-0.0387],[449,0.0183],[450,0.0994],[451,0.2026],[452,0.3261],[453,0.4681],[454,0.6265],[455,0.7995],[456,0.9848],[457,1.1805],[458,1.3843],[459,1.5941],[463,2.4492],[464,2.6563],[465,2.8568],[466,3.0487],[467,3.2304],[468,3.4002],[469,3.5592],[470,3.7097],[471,3.8514],[472,3.9846],[473,4.1089],[474,4.2242],[475,4.3303],[476,4.4268],[477,4.5135],[478,4.5899],[479,4.6557],[480,4.7104],[481,4.7536],[482,4.785],[483,4.804],[484,4.8104],[485,-2.0071],[486,-2.7267],[487,-3.4143],[488,-3.4099],[489,-3.347],[490,-3.2316],[491,-3.0699],[492,-2.8688],[493,-2.6366],[494,-2.3824],[496,-1.8511],[497,-1.607],[498,-1.4082],[499,-1.2699],[500,-1.2071],[502,-1.1642],[505,-1.0911],[508,-1.009],[512,-0.8864],[513,-0.8573],[514,-0.8306],[515,-0.8074],[516,-0.7886],[517,-0.8338],[518,-0.9798],[519,-1.1932],[521,-1.6909],[522,-1.9111],[523,-2.0924],[524,-2.2112],[525,-0.6362],[526,2.5807],[527,4.6456],[528,4.9624],[529,4.7914],[530,4.3904],[531,3.8601],[532,3.3019],[533,2.8222],[534,2.5264],[535,2.3581],[536,2.1962],[537,2.0493],[538,1.9256],[539,1.8332],[540,1.7802],[541,1.7548],[542,1.741],[544,1.7314],[553,1.7313],[561,1.7146],[576,1.7146]]
       }
       CTRL_C_eyesAim.Rotation.X {
        type=MovieSceneScriptingFloatChannel
        default=0
        thinned_from=575
        keys [[0,-7.1045],[19,-7.0273],[37,-7.1557],[38,-7.3148],[40,-7.991],[41,-8.1432],[44,13.1421],[50,13.3531],[51,13.1654],[52,12.6145],[53,11.8445],[54,10.999],[55,10.2217],[56,9.6563],[58,8.8312],[59,8.4384],[60,8.0719],[61,7.7405],[62,7.4529],[63,7.218],[64,7.0445],[65,6.9411],[77,6.1826],[81,5.9778],[84,5.8575],[87,5.7744],[91,5.7265],[92,5.7877],[93,5.9646],[94,6.2316],[95,6.5631],[98,7.6891],[99,8.0229],[100,8.2932],[101,8.4745],[102,8.5708],[103,-15.1812],[104,-14.7802],[105,-13.8402],[106,-12.7188],[107,-11.7733],[108,-11.0505],[109,-10.4206],[110,-9.9767],[112,-9.3473],[113,-9.0517],[114,-8.7789],[115,-8.5362],[116,-8.331],[117,-8.1709],[118,-8.0631],[125,-7.5373],[129,-7.2996],[132,-7.1777],[150,-6.8004],[159,-6.7567],[164,-6.7831],[165,-6.8315],[166,-6.9574],[167,-7.1556],[168,-7.4205],[169,-7.7469],[170,-8.1294],[171,-8.5627],[172,-9.0382],[173,-9.5478],[174,-10.0872],[176,-11.2368],[183,-15.5273],[185,-16.6797],[187,-17.7342],[188,-18.2119],[189,-18.6473],[190,-19.0355],[191,-19.3711],[192,-19.6492],[193,-19.8647],[194,-20.0124],[195,-20.0872],[196,-20.0843],[197,-20.0151],[198,-19.8975],[199,-19.7357],[201,-19.2968],[203,-18.7325],[206,-17.725],[212,-15.5483],[214,-14.8973],[216,-14.3386],[217,-14.1063],[218,-13.9127],[219,-13.7605],[238,-12.1788],[242,-11.9438],[261,-11.6233],[263,-11.6428],[266,-11.7654],[271,-12.0729],[273,-12.1412],[275,-12.1529],[277,-12.107],[283,-11.8113],[285,-11.6632],[287,-11.4561],[289,-11.1739],[290,-11.0001],[302,-8.4456],[305,-7.8636],[307,-7.5102],[309,-7.1916],[311,-6.9138],[313,-6.6829],[314,-6.5869],[315,-6.5048],[316,-6.4375],[317,-6.3857],[326,-6.1573],[329,-6.1085],[330,-6.1085],[331,-6.0897],[346,-6.193],[347,-6.193],[348,-6.2118],[349,-6.2118],[353,-6.2278],[364,-6.0943],[368,-5.9804],[372,-5.8161],[389,-4.8356],[392,-4.6877],[394,-4.6139],[396,-4.5718],[398,-4.5605],[400,-4.5789],[402,-4.6261],[403,-4.6602],[404,-4.7352],[405,-4.879],[406,-5.0829],[407,-5.3378],[408,-5.6349],[409,-5.9653],[410,-6.3202],[414,-7.8095],[415,-8.1557],[416,-8.474],[417,-8.666],[418,-8.6883],[420,-8.5409],[423,-8.4169],[424,-8.3987],[425,-8.4042],[429,-8.565],[431,-8.5956],[433,-8.5762],[435,-8.5256],[436,-8.5807],[437,-8.7624],[438,-9.0439],[439,-9.3983],[442,-10.6215],[443,-10.9878],[444,-11.2858],[445,-11.4872],[446,-11.5634],[447,-11.5435],[448,-11.4668],[449,-11.3269],[450,-11.1286],[451,-10.8762],[452,-10.5743],[453,-10.2276],[454,-9.8408],[455,-9.4188],[456,-8.9665],[457,-8.4893],[459,-7.4812],[463,-5.4008],[464,-4.8979],[465,-4.4115],[466,-3.9463],[467,-3.5062],[468,-3.0954],[469,-2.711],[470,-2.3481],[471,-2.0064],[472,-1.6861],[473,-1.3874],[474,-1.1107],[475,-0.8565],[476,-0.6254],[477,-0.4182],[478,-0.2357],[479,-0.0788],[480,0.0517],[481,0.1547],[482,0.2293],[483,0.2747],[484,0.2901],[485,-16.7217],[486,-18.6681],[487,-20.557],[488,-20.525],[489,-20.3222],[490,-19.971],[491,-19.4946],[492,-18.9175],[493,-18.2658],[496,-16.1452],[497,-15.5043],[498,-14.9848],[499,-14.6237],[500,-14.4576],[515,-13.4357],[516,-13.3916],[517,-13.5103],[518,-13.8862],[519,-14.4362],[521,-15.7264],[522,-16.3007],[523,-16.7668],[524,-17.0582],[525,-12.9859],[526,-5.0411],[527,-0.0985],[528,0.6529],[529,0.2476],[530,-0.7043],[531,-1.9669],[532,-3.3001],[533,-4.4495],[534,-5.1591],[535,-5.5624],[536,-5.9501],[537,-6.3021],[538,-6.5984],[539,-6.8194],[540,-6.9457],[541,-7.006],[542,-7.0389],[553,-7.0654],[576,-7.1048]]
       }
       CTRL_C_eyesAim.Rotation.Y {
        type=MovieSceneScriptingFloatChannel
        default=0
        thinned_from=575
        keys [[0,-7.5335],[18,-7.6282],[36,-7.4508],[37,-7.4508],[38,-7.5004],[40,-7.7588],[41,-7.8251],[44,-8.1026],[46,-8.0139],[52,-7.6065],[54,-7.5129],[56,-7.472],[101,-7.5284],[102,-7.4932],[103,-7.9953],[104,-8.0097],[171,-7.7123],[173,-7.7391],[175,-7.8095],[177,-7.9273],[179,-8.0931],[181,-8.3033],[193,-9.8414],[196,-10.1541],[199,-10.3813],[202,-10.5251],[205,-10.5862],[209,-10.5543],[217,-10.2576],[227,-9.7546],[230,-9.6726],[236,-9.6617],[260,-10.2581],[266,-10.3251],[268,-10.2984],[270,-10.2268],[273,-10.0405],[278,-9.5905],[279,-9.4795],[281,-9.1966],[286,-8.3818],[287,-8.2409],[288,-8.1182],[289,-8.0178],[290,-7.9435],[291,-7.8985],[303,-7.6307],[352,-7.5265],[368,-7.6733],[395,-7.5351],[404,-7.353],[407,-7.2434],[414,-6.8823],[416,-6.8026],[417,-6.8914],[418,-7.1501],[419,-7.4622],[420,-7.708],[422,-8.0475],[423,-8.1997],[424,-8.3297],[425,-8.4298],[426,-8.4917],[432,-8.7061],[436,-8.794],[443,-9.0907],[445,-9.1292],[448,-9.1331],[449,-9.1643],[451,-9.2805],[456,-9.6619],[458,-9.7592],[460,-9.787],[461,-9.771],[462,-9.7343],[464,-9.6014],[466,-9.3992],[472,-8.6162],[474,-8.3841],[476,-8.1895],[478,-8.0375],[480,-7.9294],[482,-7.8646],[484,-7.8427],[485,-12.6948],[486,-13.2631],[487,-13.713],[488,-13.55],[489,-13.27],[490,-12.9075],[493,-11.6546],[494,-11.2769],[495,-10.957],[496,-10.7112],[497,-10.5186],[498,-10.3532],[499,-10.2205],[500,-10.1253],[504,-9.8708],[507,-9.7618],[509,-9.7432],[511,-9.7778],[516,-9.9967],[517,-10.0722],[518,-10.1877],[521,-10.6157],[522,-10.7257],[523,-10.7089],[524,-10.5144],[525,-9.6315],[526,-8.3777],[527,-7.7158],[528,-7.6231],[529,-7.6656],[531,-7.865],[532,-7.9474],[533,-7.9797],[534,-7.9518],[536,-7.8151],[539,-7.5794],[540,-7.5192],[541,-7.4796],[542,-7.4651],[552,-7.563],[576,-7.5358]]
       }
       CTRL_C_eyesAim.Rotation.Z {
        type=MovieSceneScriptingFloatChannel
        default=0
        thinned_from=575
        keys [[0,-10.9001],[1,-10.9001],[4,-11.0095],[13,-11.5717],[15,-11.6511],[18,-11.6732],[20,-11.592],[22,-11.4521],[31,-10.5326],[33,-10.3685],[35,-10.2573],[36,-10.2269],[37,-10.2269],[38,-10.6632],[40,-12.6261],[41,-13.0713],[44,-9.8857],[45,-10.0525],[46,-10.2684],[47,-10.5229],[49,-11.107],[52,-12.0179],[53,-12.292],[54,-12.5345],[55,-12.7346],[56,-12.8809],[57,-12.9628],[68,-13.3886],[74,-13.4501],[83,-13.2197],[96,-12.7151],[101,-12.6444],[103,-13.1179],[106,-13.3689],[109,-13.4797],[149,-13.3215],[153,-13.1956],[166,-12.3932],[169,-12.2872],[171,-12.2639],[172,-12.2934],[173,-12.3792],[174,-12.5153],[175,-12.6961],[176,-12.9156],[177,-13.1682],[179,-13.7493],[185,-15.6949],[198,-19.0582],[204,-20.3801],[208,-21.0778],[211,-21.462],[214,-21.6982],[216,-21.7614],[218,-21.6994],[220,-21.5019],[226,-20.6385],[228,-20.4352],[230,-20.36],[233,-20.4404],[237,-20.7373],[256,-23.2776],[261,-23.7215],[265,-23.8422],[266,-23.8114],[267,-23.726],[269,-23.4104],[271,-22.9334],[273,-22.3325],[278,-20.5242],[279,-20.0857],[280,-19.5441],[281,-18.9216],[283,-17.523],[285,-16.0683],[286,-15.3754],[287,-14.7352],[288,-14.17],[289,-13.7021],[290,-13.3539],[291,-13.1479],[301,-11.9277],[306,-11.4289],[310,-11.1272],[313,-10.974],[316,-10.8951],[332,-10.9196],[346,-10.63],[349,-10.6116],[350,-10.6199],[352,-10.6991],[354,-10.8404],[363,-11.7697],[365,-11.9355],[367,-12.0481],[369,-12.0896],[373,-12.0344],[378,-11.8316],[388,-11.1304],[392,-10.796],[393,-10.6932],[394,-10.5559],[395,-10.3875],[396,-10.1917],[397,-9.9717],[399,-9.4732],[406,-7.4801],[407,-7.2111],[408,-6.9566],[409,-6.72],[410,-6.5049],[411,-6.3147],[412,-6.1526],[413,-6.022],[414,-5.9262],[415,-5.8682],[416,-5.8513],[417,-6.6338],[418,-8.4649],[419,-10.5857],[420,-12.2366],[421,-13.3898],[422,-14.4908],[423,-15.4869],[424,-16.3254],[425,-16.9535],[426,-17.3186],[431,-18.3463],[434,-18.8317],[436,-19.0559],[438,-19.1689],[440,-19.1556],[446,-18.8311],[447,-18.8184],[448,-18.8975],[449,-19.1226],[450,-19.4787],[451,-19.9508],[452,-20.5239],[453,-21.1831],[454,-21.9134],[456,-23.5264],[460,-26.929],[462,-28.4672],[463,-29.14],[464,-29.7271],[465,-30.2129],[466,-30.5815],[467,-30.817],[468,-30.9039],[469,-30.8667],[470,-30.7479],[472,-30.3118],[475,-29.3358],[479,-27.921],[481,-27.3734],[482,-27.18],[483,-27.0555],[484,-27.0115],[485,-27.4125],[486,-27.3464],[488,-26.6796],[489,-26.1684],[490,-25.5428],[493,-23.434],[494,-22.8107],[495,-22.3095],[496,-21.9821],[501,-20.8379],[503,-20.4523],[505,-20.1681],[507,-20.0146],[508,-19.996],[509,-20.0286],[510,-20.1152],[512,-20.4143],[517,-21.4679],[519,-21.7841],[520,-21.8825],[521,-21.9339],[522,-21.9342],[523,-21.613],[526,-19.3836],[527,-18.8035],[530,-17.6806],[531,-17.2183],[532,-16.6072],[533,-15.8852],[534,-15.091],[536,-13.4245],[537,-12.6218],[538,-11.8865],[539,-11.2535],[540,-10.758],[541,-10.4348],[542,-10.3193],[543,-10.3423],[544,-10.4046],[546,-10.6078],[549,-10.9617],[550,-11.0534],[551,-11.1157],[552,-11.1386],[553,-11.1386],[561,-10.9191],[576,-10.9191]]
       }
       CTRL_C_eyesAim.Scale.X {type=MovieSceneScriptingFloatChannel, default=1, const [1,575]}
       CTRL_C_eyesAim.Scale.Y {type=MovieSceneScriptingFloatChannel, default=1, const [1,575]}
       CTRL_C_eyesAim.Scale.Z {type=MovieSceneScriptingFloatChannel, default=1, const [1,575]}
       CTRL_L_eyeAim.Location.X {type=MovieSceneScriptingFloatChannel, default=0, const [0,577]}
       CTRL_L_eyeAim.Location.Y {type=MovieSceneScriptingFloatChannel, default=0, const [0,577]}
       CTRL_L_eyeAim.Location.Z {type=MovieSceneScriptingFloatChannel, default=0, const [0,577]}
       CTRL_L_eyeAim.Rotation.X {type=MovieSceneScriptingFloatChannel, default=0, const [0,577]}
       CTRL_L_eyeAim.Rotation.Y {type=MovieSceneScriptingFloatChannel, default=0, const [0,577]}
       CTRL_L_eyeAim.Rotation.Z {type=MovieSceneScriptingFloatChannel, default=0, const [0,577]}
       CTRL_L_eyeAim.Scale.X {type=MovieSceneScriptingFloatChannel, default=1, const [1,577]}
       CTRL_L_eyeAim.Scale.Y {type=MovieSceneScriptingFloatChannel, default=1, const [1,577]}
       CTRL_L_eyeAim.Scale.Z {type=MovieSceneScriptingFloatChannel, default=1, const [1,577]}
       CTRL_R_eyeAim.Location.X {
        type=MovieSceneScriptingFloatChannel
        default=0
        thinned_from=577
        keys [[0,0.0654],[1,0.0654],[3,0.0637],[4,0.0625],[6,0.0595],[11,0.0506],[13,0.0476],[14,0.0464],[15,0.0455],[16,0.0449],[18,0.0449],[19,0.0475],[20,0.0506],[21,0.0547],[22,0.0594],[23,0.0647],[26,0.0812],[27,0.0864],[28,0.091],[29,0.0948],[30,0.0977],[31,0.0995],[32,0.0999],[33,0.0987],[34,0.0957],[35,0.0907],[36,0.0835],[37,-0.1627],[38,-0.4754],[39,-0.2992],[40,-0.0992],[41,-0.0077],[42,0.1462],[43,0.144],[44,0.1378],[45,0.1283],[46,0.1159],[47,0.1012],[48,0.085],[49,0.0676],[51,0.0322],[52,0.0155],[53,0.0006],[54,-0.0117],[55,-0.0213],[56,-0.028],[57,-0.0314],[66,-0.0444],[68,-0.0467],[70,-0.0482],[71,-0.0487],[72,-0.0489],[73,-0.0489],[74,-0.0486],[75,-0.048],[76,-0.0473],[78,-0.0452],[80,-0.0424],[83,-0.0374],[91,-0.0224],[92,-0.0208],[93,-0.0195],[95,-0.0174],[97,-0.0161],[99,-0.0154],[101,-0.0154],[102,0.0597],[103,0.181],[104,0.1718],[105,0.1483],[106,0.113],[107,0.0691],[108,0.0246],[109,-0.0129],[110,-0.0411],[111,-0.0537],[112,-0.0549],[113,-0.055],[114,-0.0542],[115,-0.0525],[116,-0.0501],[117,-0.0472],[118,-0.0438],[121,-0.0331],[122,-0.03],[123,-0.0272],[124,-0.0251],[125,-0.0238],[128,-0.0224],[134,-0.0204],[135,-0.0205],[136,-0.0213],[138,-0.0239],[139,-0.0243],[140,-0.0241],[145,-0.0244],[147,-0.0242],[148,-0.0219],[149,-0.0177],[150,-0.0139],[151,-0.0106],[152,-0.0077],[153,-0.0051],[154,-0.003],[155,-0.0011],[156,0.0004],[157,0.0017],[158,0.0027],[159,0.0034],[161,0.0043],[163,0.0046],[169,0.0045],[171,0.0049],[172,0.0045],[173,0.0028],[174,0.0002],[175,-0.0031],[178,-0.0143],[179,-0.0177],[180,-0.0205],[181,-0.0227],[182,-0.024],[183,-0.0246],[184,-0.0242],[185,-0.023],[186,-0.0209],[187,-0.018],[191,-0.004],[192,-0.0008],[193,0.0021],[194,0.0045],[195,0.0062],[196,0.0071],[197,0.0072],[198,0.0066],[199,0.0053],[200,0.0034],[201,0.0009],[202,-0.0023],[203,-0.006],[204,-0.0103],[205,-0.015],[206,-0.0202],[207,-0.0257],[209,-0.0375],[212,-0.0556],[213,-0.0614],[214,-0.0668],[215,-0.0719],[217,-0.0809],[221,-0.1012],[224,-0.1175],[225,-0.1226],[226,-0.1274],[227,-0.1317],[228,-0.1353],[229,-0.138],[230,-0.1399],[232,-0.143],[234,-0.1472],[236,-0.1522],[239,-0.1609],[249,-0.1922],[253,-0.2033],[257,-0.2126],[260,-0.218],[261,-0.2193],[262,-0.2202],[263,-0.2205],[264,-0.2203],[265,-0.2196],[266,-0.2179],[267,-0.2148],[268,-0.2105],[269,-0.2051],[270,-0.1989],[271,-0.192],[272,-0.1846],[274,-0.1687],[278,-0.1344],[279,-0.1249],[280,-0.1142],[281,-0.1024],[282,-0.0898],[283,-0.0766],[285,-0.0499],[286,-0.037],[287,-0.0249],[288,-0.0141],[289,-0.0051],[290,0.0017],[291,0.0057],[294,0.0132],[304,0.0411],[306,0.0463],[308,0.0512],[310,0.0554],[311,0.0572],[312,0.0588],[313,0.0602],[314,0.0614],[315,0.0622],[316,0.0628],[323,0.0634],[326,0.0631],[328,0.0625],[329,0.0631],[330,0.0631],[331,0.0663],[332,0.0663],[333,0.0684],[334,0.0689],[336,0.0694],[342,0.0685],[344,0.0686],[345,0.0689],[346,0.0694],[347,0.0748],[348,0.0674],[349,0.074],[350,0.0772],[351,0.0851],[352,0.0897],[353,0.0921],[354,0.0925],[355,0.0911],[356,0.0882],[357,0.0841],[358,0.0789],[359,0.0729],[360,0.0664],[363,0.0458],[364,0.0394],[365,0.0337],[366,0.0288],[367,0.0249],[368,0.0225],[369,0.0216],[370,0.0217],[371,0.022],[372,0.0226],[373,0.0233],[375,0.0255],[377,0.0283],[379,0.0319],[381,0.036],[383,0.0407],[386,0.0485],[392,0.066],[393,0.0701],[394,0.0762],[395,0.0839],[396,0.0931],[397,0.1035],[398,0.1147],[399,0.1266],[402,0.1631],[403,0.1747],[404,0.1852],[405,0.1943],[406,0.2016],[407,0.2069],[408,0.2098],[409,0.2102],[410,0.2078],[411,0.2024],[412,0.194],[413,0.1824],[414,0.1673],[415,0.1488],[416,0.0682],[417,0.1868],[418,0.6277],[419,1.1066],[420,1.3384],[421,1.3332],[422,1.2918],[423,1.2144],[424,1.1012],[425,0.9525],[426,0.7684],[427,0.616],[429,0.4166],[430,0.3192],[431,0.2254],[432,0.1366],[433,0.0544],[434,-0.0195],[435,-0.0835],[436,-0.1352],[437,-0.1723],[438,-0.1936],[439,-0.1979],[440,-0.1923],[441,-0.1851],[443,-0.1696],[444,-0.1627],[445,-0.1575],[446,-0.1546],[447,-0.154],[448,-0.1557],[449,-0.1597],[450,-0.1658],[451,-0.174],[452,-0.1842],[453,-0.1964],[454,-0.2105],[455,-0.2264],[456,-0.244],[457,-0.2631],[458,-0.2837],[459,-0.3054],[463,-0.3975],[464,-0.4199],[465,-0.4411],[466,-0.4605],[467,-0.4776],[468,-0.4917],[469,-0.5027],[470,-0.511],[471,-0.5166],[472,-0.5198],[473,-0.5208],[474,-0.5198],[475,-0.5172],[476,-0.5131],[481,-0.4852],[482,-0.481],[483,-0.4781],[484,-0.4771],[485,0.0804],[486,-0.0001],[487,0.0388],[488,0.0045],[489,-0.0473],[490,-0.1083],[491,-0.1701],[492,-0.2245],[493,-0.2637],[494,-0.2805],[495,-0.2681],[496,-0.22],[497,-0.1836],[498,-0.1788],[499,-0.1699],[500,-0.157],[501,-0.1423],[502,-0.128],[503,-0.1146],[504,-0.1027],[505,-0.0928],[506,-0.0853],[507,-0.0809],[508,-0.08],[509,-0.0813],[510,-0.0833],[511,-0.0858],[512,-0.0887],[515,-0.0984],[516,-0.1014],[517,-0.1018],[518,-0.0975],[519,-0.0897],[520,-0.0798],[521,-0.0692],[522,-0.0594],[523,-0.0502],[524,-0.0428],[525,-0.096],[526,-0.1829],[527,-0.2213],[528,-0.2157],[529,-0.1981],[530,-0.176],[531,-0.1494],[533,-0.089],[534,-0.0612],[535,-0.035],[536,-0.0094],[537,0.0146],[538,0.0361],[539,0.0543],[540,0.0684],[541,0.0776],[542,0.0808],[543,0.0802],[544,0.0785],[545,0.0761],[546,0.0731],[548,0.0667],[549,0.0637],[550,0.0613],[551,0.0596],[552,0.059],[553,0.059],[555,0.0603],[559,0.0637],[560,0.0644],[561,0.0649],[576,0.0649]]
       }
       CTRL_R_eyeAim.Location.Y {
        type=MovieSceneScriptingFloatChannel
        default=0
        thinned_from=577
        keys [[0,1.5171],[37,1.5216],[38,1.5079],[40,1.4143],[41,1.3941],[42,3.2596],[50,3.2618],[51,3.2442],[52,3.1983],[53,3.1351],[55,3.0012],[56,2.9539],[60,2.8222],[62,2.7701],[64,2.7352],[85,2.6348],[91,2.6295],[92,2.6351],[93,2.6504],[95,2.7013],[98,2.7956],[100,2.8456],[101,2.8605],[102,2.866],[103,0.7686],[104,0.8079],[105,0.9],[106,1.006],[107,1.0871],[108,1.1432],[109,1.1934],[110,1.2294],[113,1.3072],[115,1.353],[116,1.3717],[117,1.3865],[127,1.4581],[132,1.4795],[164,1.5271],[165,1.5235],[166,1.5125],[167,1.4948],[168,1.4708],[169,1.4409],[170,1.4057],[171,1.3656],[172,1.321],[173,1.2724],[174,1.2203],[175,1.1651],[177,1.0474],[183,0.6714],[184,0.6107],[185,0.552],[186,0.4957],[187,0.4423],[188,0.3925],[189,0.3467],[190,0.3053],[191,0.2689],[192,0.238],[193,0.2131],[194,0.1947],[195,0.1833],[196,0.1793],[197,0.1816],[198,0.1887],[199,0.2],[200,0.2152],[201,0.2339],[202,0.2557],[203,0.2802],[204,0.307],[205,0.3358],[206,0.3661],[208,0.4298],[211,0.5274],[213,0.5894],[214,0.6184],[215,0.6455],[216,0.6704],[217,0.6927],[218,0.7121],[219,0.7282],[220,0.7406],[228,0.8213],[233,0.8629],[237,0.888],[240,0.9008],[243,0.9076],[262,0.8885],[269,0.8643],[271,0.8629],[272,0.8646],[274,0.8734],[276,0.8886],[278,0.9096],[280,0.9363],[282,0.9679],[286,1.0416],[303,1.4019],[307,1.4744],[310,1.5197],[313,1.555],[316,1.5785],[329,1.6084],[365,1.5996],[374,1.6313],[395,1.7532],[402,1.7684],[403,1.7682],[404,1.7642],[405,1.7539],[406,1.7379],[407,1.7172],[409,1.6647],[415,1.4749],[416,1.4463],[417,1.4219],[418,1.4082],[420,1.4084],[422,1.3851],[426,1.3208],[427,1.3076],[429,1.2888],[431,1.2761],[433,1.2704],[435,1.2698],[436,1.2629],[437,1.2447],[438,1.2177],[439,1.1844],[442,1.0717],[443,1.0384],[444,1.0114],[445,0.9933],[446,0.9868],[447,0.9889],[448,0.9947],[449,1.0041],[450,1.0166],[451,1.032],[452,1.0501],[453,1.0707],[455,1.1179],[457,1.1717],[461,1.2923],[466,1.4581],[473,1.7053],[475,1.7689],[477,1.8247],[479,1.8702],[481,1.9029],[482,1.9137],[483,1.9204],[484,1.9227],[485,0.5869],[486,0.1389],[487,-0.0351],[488,-0.0242],[489,0.0067],[490,0.0543],[491,0.1153],[492,0.1863],[493,0.2635],[494,0.3433],[495,0.4219],[496,0.4954],[497,0.5608],[498,0.6145],[499,0.653],[500,0.6726],[505,0.7186],[508,0.742],[510,0.7543],[512,0.7633],[514,0.7687],[516,0.7699],[517,0.7547],[518,0.7157],[519,0.6606],[521,0.5347],[522,0.4798],[523,0.4414],[524,0.4276],[525,0.8297],[526,1.585],[527,2.0416],[528,2.1142],[529,2.0842],[530,2.0054],[531,1.8995],[532,1.7884],[533,1.6947],[534,1.6414],[536,1.5918],[537,1.5699],[538,1.5519],[539,1.5392],[540,1.5334],[576,1.5168]]
       }
       CTRL_R_eyeAim.Location.Z {
        type=MovieSceneScriptingFloatChannel
        default=0
        thinned_from=577
        keys [[0,0.1995],[36,0.2005],[37,0.0832],[38,-0.0573],[39,0.0305],[40,0.1283],[41,0.1728],[42,1.0731],[50,1.0731],[51,1.0589],[52,1.0225],[53,0.9739],[54,0.9225],[55,0.8772],[56,0.8461],[58,0.8047],[60,0.7676],[61,0.7516],[62,0.7379],[63,0.7267],[64,0.7184],[65,0.7135],[79,0.6738],[86,0.6631],[91,0.6612],[92,0.6642],[93,0.6723],[94,0.6845],[95,0.6998],[98,0.7526],[99,0.7686],[100,0.7816],[101,0.7904],[102,0.7937],[103,-1.2153],[104,-1.2547],[105,-1.3168],[106,-1.3182],[107,-1.1775],[108,-0.9405],[109,-0.7244],[110,-0.5567],[111,-0.4231],[112,-0.2934],[113,-0.1719],[114,-0.0628],[115,0.0298],[116,0.1017],[117,0.1487],[118,0.1664],[126,0.1801],[130,0.1852],[153,0.1975],[164,0.1995],[165,0.1985],[166,0.1954],[167,0.1906],[168,0.1842],[169,0.1764],[170,0.1674],[171,0.1574],[172,0.1467],[173,0.1355],[176,0.101],[177,0.0897],[178,0.0789],[179,0.0686],[180,0.059],[181,0.0501],[182,0.042],[183,0.0348],[184,0.0284],[185,0.0229],[186,0.0181],[187,0.0142],[188,0.011],[189,0.0084],[190,0.0065],[191,0.005],[192,0.0039],[193,0.0032],[194,0.0027],[196,0.0024],[198,0.0026],[200,0.0034],[202,0.0048],[203,0.0058],[204,0.0071],[205,0.0085],[206,0.0103],[207,0.0123],[208,0.0145],[209,0.017],[211,0.0226],[216,0.0379],[217,0.0407],[218,0.0432],[219,0.0454],[220,0.0471],[230,0.062],[234,0.0672],[237,0.0704],[239,0.0721],[241,0.0734],[243,0.0742],[245,0.0745],[252,0.0738],[256,0.0726],[259,0.0712],[260,0.0712],[261,0.0719],[262,0.076],[263,0.0857],[264,0.0996],[265,0.1091],[266,0.1085],[267,0.1009],[268,0.0895],[269,0.0775],[270,0.0679],[271,0.064],[272,0.0641],[273,0.0645],[274,0.0652],[275,0.0661],[276,0.0672],[277,0.0686],[278,0.0702],[279,0.072],[280,0.074],[281,0.0763],[282,0.0788],[283,0.0815],[284,0.0844],[286,0.0908],[288,0.0979],[290,0.1058],[292,0.1143],[295,0.1281],[305,0.1771],[308,0.1904],[310,0.1983],[312,0.2052],[314,0.2107],[316,0.2147],[317,0.2161],[328,0.2229],[346,0.2232],[347,0.2197],[348,0.2263],[349,0.2195],[350,0.2257],[351,0.2222],[358,0.22],[359,0.2201],[360,0.2216],[361,0.225],[362,0.2302],[363,0.2374],[364,0.2453],[365,0.2525],[366,0.259],[367,0.2649],[368,0.2702],[369,0.2749],[371,0.2827],[373,0.2885],[375,0.2925],[377,0.295],[379,0.296],[382,0.2954],[386,0.2916],[397,0.2754],[400,0.2725],[403,0.2714],[404,0.2788],[405,0.2997],[406,0.3317],[407,0.3726],[408,0.4199],[409,0.4715],[411,0.5785],[412,0.6292],[413,0.675],[414,0.7136],[415,0.7427],[416,0.7599],[417,0.7631],[418,0.7496],[419,0.7169],[420,0.6624],[421,0.597],[422,0.534],[423,0.4738],[424,0.4169],[425,0.364],[426,0.3156],[427,0.2722],[428,0.2341],[429,0.202],[430,0.1761],[431,0.1571],[432,0.1454],[433,0.1414],[435,0.1416],[436,0.1401],[437,0.136],[438,0.1299],[439,0.1226],[441,0.1068],[442,0.0995],[443,0.0931],[444,0.0881],[445,0.0849],[446,0.0837],[447,0.084],[448,0.085],[449,0.0867],[450,0.0889],[451,0.0918],[452,0.0953],[453,0.0994],[454,0.1041],[455,0.1094],[456,0.1153],[457,0.1217],[458,0.1288],[459,0.1364],[460,0.1445],[461,0.1531],[462,0.1622],[463,0.1717],[465,0.192],[472,0.268],[474,0.2886],[476,0.3073],[477,0.3157],[478,0.3233],[479,0.33],[480,0.3358],[481,0.3404],[482,0.3438],[483,0.3459],[484,0.3466],[485,3.6422],[486,0.02],[487,0.023],[488,0.0159],[489,-0.0031],[490,-0.0305],[491,-0.0628],[492,-0.0964],[493,-0.1278],[494,-0.1536],[495,-0.1706],[496,-0.1757],[497,-0.1662],[498,-0.1393],[499,-0.0926],[500,-0.0234],[501,0.0203],[502,0.0243],[504,0.0332],[507,0.047],[508,0.0514],[509,0.0556],[510,0.0595],[511,0.063],[512,0.0661],[513,0.0686],[514,0.0705],[515,0.0717],[516,0.0721],[517,0.0702],[518,0.0653],[519,0.059],[520,0.0524],[521,0.0466],[522,0.0421],[523,0.0393],[524,0.0383],[525,0.0737],[526,0.2243],[527,0.3716],[528,0.3987],[529,0.3864],[530,0.3559],[531,0.3174],[532,0.2798],[533,0.2502],[534,0.2341],[535,0.2266],[536,0.2197],[537,0.2136],[538,0.2088],[539,0.2055],[540,0.204],[552,0.1997],[576,0.1994]]
       }
       CTRL_R_eyeAim.Rotation.X {
        type=MovieSceneScriptingFloatChannel
        default=0
        thinned_from=577
        keys [[0,-0.0333],[35,-0.0332],[36,-0.0331],[37,-0.3086],[38,-0.6252],[39,-0.3906],[40,-0.1334],[41,-0.0166],[42,0],[50,0],[51,-0.003],[52,-0.0104],[53,-0.0199],[54,-0.0291],[55,-0.0357],[56,-0.0372],[59,-0.0323],[61,-0.0294],[62,-0.0282],[63,-0.0273],[64,-0.0266],[86,-0.0218],[91,-0.0214],[92,-0.0217],[93,-0.0223],[94,-0.0232],[99,-0.0293],[100,-0.0302],[101,-0.0309],[102,-0.0311],[103,-2.97],[104,-3.0744],[105,-3.2514],[106,-3.2962],[107,-3.0042],[108,-2.4791],[109,-2.0007],[110,-1.6285],[111,-1.3307],[112,-1.0423],[113,-0.7727],[114,-0.5312],[115,-0.3271],[116,-0.1699],[117,-0.0688],[118,-0.0333],[128,-0.037],[164,-0.04],[165,-0.0399],[167,-0.0391],[169,-0.0375],[171,-0.0353],[173,-0.0326],[177,-0.0261],[185,-0.0119],[188,-0.0074],[190,-0.0049],[192,-0.003],[194,-0.0018],[195,-0.0015],[197,-0.0014],[200,-0.0021],[213,-0.0094],[216,-0.0105],[219,-0.0109],[252,-0.0075],[254,-0.0077],[256,-0.0085],[257,-0.0092],[258,-0.01],[259,-0.0112],[260,-0.0109],[261,-0.0092],[262,0.0007],[263,0.0245],[264,0.0585],[265,0.0824],[266,0.0826],[267,0.0665],[268,0.0413],[269,0.0143],[270,-0.0073],[271,-0.0161],[275,-0.0167],[280,-0.0184],[307,-0.0362],[313,-0.039],[318,-0.0401],[334,-0.0411],[346,-0.0382],[347,-0.0463],[348,-0.0299],[349,-0.0458],[350,-0.0302],[351,-0.0379],[354,-0.0387],[356,-0.0397],[358,-0.041],[359,-0.0406],[360,-0.037],[361,-0.0292],[362,-0.017],[363,-0.0002],[364,0.0178],[365,0.0339],[366,0.048],[367,0.0603],[368,0.0708],[369,0.0796],[370,0.0867],[371,0.0923],[372,0.0965],[373,0.0993],[374,0.1008],[375,0.1011],[376,0.1003],[377,0.0984],[378,0.0955],[379,0.0918],[380,0.0872],[381,0.082],[382,0.0761],[383,0.0696],[384,0.0627],[385,0.0553],[386,0.0477],[391,0.0075],[392,-0.0003],[393,-0.0079],[394,-0.0152],[395,-0.022],[396,-0.0283],[397,-0.0341],[398,-0.0391],[399,-0.0435],[400,-0.047],[401,-0.0496],[402,-0.0512],[403,-0.0517],[404,-0.0314],[405,0.0254],[406,0.1126],[407,0.2242],[408,0.354],[409,0.4961],[411,0.7927],[412,0.9353],[413,1.066],[414,1.1789],[415,1.268],[416,1.3275],[417,1.3518],[418,1.3345],[419,1.269],[420,1.1489],[421,1.0017],[422,0.8597],[423,0.7242],[424,0.5963],[425,0.4774],[426,0.3686],[427,0.271],[428,0.1856],[429,0.1135],[430,0.0556],[431,0.013],[432,-0.0134],[433,-0.0223],[436,-0.0222],[438,-0.0212],[443,-0.0176],[445,-0.0167],[446,-0.0166],[448,-0.0169],[450,-0.018],[452,-0.0197],[454,-0.0219],[457,-0.0259],[467,-0.0423],[473,-0.0526],[476,-0.057],[478,-0.0594],[480,-0.0613],[482,-0.0625],[483,-0.0628],[484,-0.0629],[485,8.565],[486,0.0426],[487,0.0536],[488,0.0372],[489,-0.0072],[490,-0.0722],[491,-0.1503],[492,-0.2339],[493,-0.3153],[494,-0.3865],[495,-0.4396],[496,-0.4666],[497,-0.4594],[498,-0.4103],[499,-0.3113],[500,-0.1546],[501,-0.0548],[502,-0.048],[503,-0.0406],[504,-0.0327],[507,-0.0081],[508,0],[509,0.0077],[510,0.0149],[511,0.0214],[512,0.0271],[513,0.0318],[514,0.0353],[515,0.0375],[516,0.0383],[517,0.0387],[518,0.04],[519,0.0418],[521,0.046],[522,0.0478],[523,0.0492],[524,0.0498],[525,0.0219],[526,-0.0316],[527,-0.0652],[528,-0.0707],[529,-0.0688],[530,-0.0639],[531,-0.0574],[532,-0.0506],[533,-0.0448],[534,-0.0414],[536,-0.0378],[537,-0.0363],[538,-0.0351],[539,-0.0342],[540,-0.0339],[576,-0.0333]]
       }
       CTRL_R_eyeAim.Rotation.Y {
        type=MovieSceneScriptingFloatChannel
        default=0
        thinned_from=577
        keys [[0,0.0006],[36,0.001],[37,0.006],[38,-0.0002],[39,-0.0036],[40,-0.0023],[41,-0.0003],[42,0],[52,0],[56,-0.0006],[102,-0.0004],[103,-0.0442],[104,-0.0515],[105,-0.0619],[106,-0.0708],[107,-0.072],[108,-0.065],[109,-0.0562],[110,-0.0478],[111,-0.0397],[112,-0.031],[113,-0.0229],[114,-0.0157],[115,-0.0096],[116,-0.0049],[117,-0.002],[118,-0.0009],[203,-0.0005],[219,-0.0019],[255,-0.0016],[259,-0.0024],[260,-0.0023],[261,-0.002],[262,0.0002],[263,0.0053],[264,0.0127],[265,0.0179],[266,0.0179],[267,0.0143],[268,0.0088],[269,0.003],[270,-0.0015],[271,-0.0032],[283,-0.002],[290,-0.0006],[346,0.0009],[347,0.001],[348,0.0007],[349,0.001],[350,0.0007],[365,0],[389,-0.0004],[396,0.0009],[403,0.0033],[404,0.0022],[405,-0.0019],[406,-0.0089],[407,-0.0187],[408,-0.031],[409,-0.0455],[410,-0.0614],[411,-0.078],[412,-0.0944],[413,-0.1096],[414,-0.1226],[415,-0.1323],[416,-0.1348],[417,-0.1265],[418,-0.1087],[419,-0.0838],[420,-0.0554],[421,-0.029],[422,-0.0078],[423,0.0077],[424,0.0174],[425,0.0218],[426,0.0219],[427,0.0188],[428,0.0143],[429,0.0096],[430,0.0051],[431,0.0013],[432,-0.0014],[433,-0.0025],[438,-0.0028],[447,-0.0021],[450,-0.0025],[453,-0.0035],[456,-0.0051],[461,-0.0091],[466,-0.0136],[468,-0.015],[471,-0.0165],[474,-0.0172],[484,-0.0168],[485,2.3733],[486,0.0123],[487,0.0154],[488,0.0105],[489,-0.002],[490,-0.0189],[491,-0.0376],[492,-0.0554],[493,-0.0705],[494,-0.0816],[495,-0.0878],[496,-0.0888],[497,-0.0844],[498,-0.0732],[499,-0.0539],[500,-0.026],[501,-0.009],[507,-0.0012],[511,0.0032],[513,0.005],[515,0.0062],[518,0.007],[522,0.0087],[523,0.0087],[524,0.0082],[525,0.0032],[526,-0.0041],[527,-0.0078],[528,-0.008],[529,-0.0074],[533,-0.0031],[536,-0.001],[539,0.0004],[541,0.0009],[576,0.0006]]
       }
       CTRL_R_eyeAim.Rotation.Z {
        type=MovieSceneScriptingFloatChannel
        default=0
        thinned_from=577
        keys [[0,0.0956],[18,0.0956],[19,0.0995],[20,0.1039],[21,0.1095],[22,0.1159],[25,0.1365],[26,0.1426],[27,0.1479],[28,0.1519],[29,0.1542],[30,0.1546],[31,0.1527],[32,0.1481],[33,0.1405],[34,0.1295],[35,0.1149],[36,0.0962],[37,-0.482],[38,-1.1882],[39,-0.7137],[40,-0.1868],[41,0.0527],[42,0.0587],[47,0.0595],[56,0.0619],[75,0.0628],[99,0.0593],[101,0.0592],[102,0.2629],[103,0.4677],[104,0.4514],[105,0.4049],[106,0.3318],[107,0.2357],[108,0.1357],[109,0.0518],[110,-0.0117],[111,-0.0399],[112,-0.0413],[113,-0.0402],[114,-0.037],[115,-0.0322],[116,-0.0259],[117,-0.0186],[118,-0.0106],[120,0.006],[121,0.014],[122,0.0213],[123,0.0276],[124,0.0325],[125,0.0358],[127,0.0381],[128,0.0397],[134,0.0455],[135,0.0453],[136,0.0432],[138,0.0367],[139,0.0357],[140,0.0361],[143,0.0358],[145,0.0352],[147,0.0356],[148,0.0411],[149,0.0507],[150,0.0585],[151,0.0647],[152,0.0694],[153,0.0727],[154,0.0747],[155,0.0756],[156,0.0754],[157,0.0744],[158,0.0726],[159,0.0701],[160,0.0671],[161,0.0637],[163,0.0563],[164,0.0563],[165,0.0488],[166,0.0453],[167,0.0422],[168,0.0395],[169,0.0375],[170,0.0362],[171,0.0357],[173,0.0353],[178,0.0327],[180,0.0322],[181,0.0324],[182,0.0328],[183,0.0336],[184,0.0348],[185,0.0366],[186,0.0388],[187,0.0416],[194,0.0633],[197,0.0719],[200,0.0799],[202,0.0848],[204,0.0891],[206,0.093],[208,0.0963],[210,0.099],[212,0.101],[214,0.1022],[215,0.1025],[216,0.1026],[217,0.1007],[218,0.0954],[219,0.0872],[220,0.0767],[221,0.0646],[222,0.0512],[224,0.0234],[225,0.0101],[226,-0.0021],[227,-0.0126],[228,-0.0208],[229,-0.0261],[230,-0.028],[231,-0.0281],[233,-0.0289],[235,-0.0304],[237,-0.0325],[239,-0.0351],[242,-0.0398],[253,-0.0595],[256,-0.0641],[258,-0.0668],[260,-0.0689],[262,-0.0704],[264,-0.0711],[265,-0.0712],[266,-0.0705],[267,-0.0683],[268,-0.0648],[269,-0.0602],[270,-0.0545],[271,-0.0478],[272,-0.0403],[273,-0.0321],[274,-0.0233],[275,-0.014],[276,-0.0044],[278,0.0155],[279,0.025],[280,0.0335],[281,0.0411],[282,0.0477],[283,0.0535],[284,0.0585],[285,0.0628],[286,0.0664],[287,0.0693],[288,0.0717],[289,0.0735],[290,0.0748],[291,0.0757],[303,0.0832],[309,0.0858],[314,0.0869],[320,0.087],[324,0.0876],[326,0.0874],[328,0.0866],[329,0.0883],[330,0.0883],[331,0.0957],[332,0.0957],[333,0.0992],[334,0.0996],[335,0.0992],[336,0.0982],[337,0.0967],[338,0.0948],[341,0.0886],[342,0.0868],[343,0.0853],[344,0.0843],[345,0.0839],[346,0.0842],[347,0.0967],[348,0.0783],[349,0.0937],[350,0.1019],[351,0.1226],[352,0.1365],[353,0.1463],[354,0.1523],[355,0.1549],[356,0.1545],[357,0.1515],[358,0.1465],[359,0.1397],[360,0.1316],[361,0.1225],[363,0.1035],[364,0.0942],[365,0.0857],[366,0.0784],[367,0.0727],[368,0.0689],[369,0.0676],[372,0.0679],[376,0.0693],[381,0.0723],[392,0.0815],[393,0.0836],[394,0.0878],[395,0.0938],[396,0.1011],[397,0.1094],[400,0.1363],[401,0.1446],[402,0.152],[403,0.1581],[404,0.1624],[405,0.1647],[406,0.1644],[407,0.1613],[408,0.1549],[409,0.1448],[410,0.1306],[411,0.112],[412,0.0886],[413,0.0599],[414,0.0257],[415,-0.0143],[416,-0.1978],[417,0.1326],[418,1.2784],[419,2.5271],[420,3.1661],[421,3.2186],[422,3.1817],[423,3.0534],[424,2.8316],[425,2.5142],[426,2.0993],[427,1.751],[429,1.3021],[430,1.0827],[431,0.8713],[432,0.6715],[433,0.4869],[434,0.3209],[435,0.1772],[436,0.0593],[437,-0.0294],[438,-0.0851],[439,-0.1045],[440,-0.1032],[441,-0.0997],[442,-0.0948],[444,-0.0836],[445,-0.0786],[446,-0.0752],[447,-0.0739],[448,-0.0729],[449,-0.07],[450,-0.0655],[451,-0.0596],[452,-0.0524],[453,-0.0441],[454,-0.035],[455,-0.0252],[456,-0.0149],[459,0.017],[460,0.0273],[461,0.0372],[462,0.0463],[463,0.0546],[464,0.0618],[465,0.0677],[466,0.0722],[467,0.075],[468,0.076],[469,0.0762],[470,0.0768],[471,0.0776],[473,0.0799],[479,0.0889],[481,0.0913],[482,0.0921],[483,0.0927],[484,0.0928],[485,0.7644],[486,0.0949],[487,0.0672],[488,-0.0053],[489,-0.1069],[490,-0.2216],[491,-0.3337],[492,-0.4273],[493,-0.4867],[494,-0.496],[495,-0.4395],[496,-0.3013],[497,-0.1937],[498,-0.1666],[499,-0.137],[500,-0.1059],[501,-0.0742],[502,-0.0432],[503,-0.0139],[504,0.0125],[505,0.0349],[506,0.0523],[507,0.0635],[508,0.0675],[522,0.0681],[523,0.0672],[524,0.0651],[525,0.0629],[526,0.0613],[527,0.0613],[530,0.0646],[531,0.0662],[532,0.0685],[533,0.0713],[534,0.0745],[537,0.085],[538,0.0882],[539,0.0909],[540,0.0931],[541,0.0946],[542,0.0951],[576,0.0956]]
       }
       CTRL_R_eyeAim.Scale.X {type=MovieSceneScriptingFloatChannel, default=1, const [1,577]}
       CTRL_R_eyeAim.Scale.Y {type=MovieSceneScriptingFloatChannel, default=1, const [1,577]}
       CTRL_R_eyeAim.Scale.Z {type=MovieSceneScriptingFloatChannel, default=1, const [1,577]}
       CTRL_convergenceSwitch {type=MovieSceneScriptingFloatChannel, default=0, const [0,577]}
       follow_head {type=MovieSceneScriptingFloatChannel, default=0, const [0,575]}
       Face_Ctrl_Master.Location.X {type=MovieSceneScriptingFloatChannel, default=0, const [0,577]}
       Face_Ctrl_Master.Location.Y {type=MovieSceneScriptingFloatChannel, default=0, const [0,577]}
       Face_Ctrl_Master.Location.Z {type=MovieSceneScriptingFloatChannel, default=0, const [0,577]}
       Face_Ctrl_Master.Rotation.X {type=MovieSceneScriptingFloatChannel, default=0, const [0,577]}
       Face_Ctrl_Master.Rotation.Y {type=MovieSceneScriptingFloatChannel, default=0, const [0,577]}
       Face_Ctrl_Master.Rotation.Z {type=MovieSceneScriptingFloatChannel, default=0, const [0,577]}
       Face_Ctrl_Master.Scale.X {type=MovieSceneScriptingFloatChannel, default=1, const [1,577]}
       Face_Ctrl_Master.Scale.Y {type=MovieSceneScriptingFloatChannel, default=1, const [1,577]}
       Face_Ctrl_Master.Scale.Z {type=MovieSceneScriptingFloatChannel, default=1, const [1,577]}
       gui_follow_head {type=MovieSceneScriptingFloatChannel, default=1, const [1,577]}
       CTRL_L_eye_blink {
        type=MovieSceneScriptingFloatChannel
        default=0
        thinned_from=577
        keys [[0,0],[35,0],[36,0.03],[37,0.096],[38,0.2976],[39,0.6549],[40,1],[41,0.9848],[42,0.9513],[43,0.8181],[44,0.2869],[45,0.0483],[46,0.0014],[47,0],[98,0],[99,0.0018],[100,0.0062],[101,0.012],[102,0.1598],[103,0.4672],[104,0.7699],[105,0.9414],[106,0.8918],[107,0.7827],[108,0.6128],[109,0.409],[110,0.273],[111,0.2054],[112,0.1415],[113,0.0853],[114,0.0404],[115,0.0107],[116,0],[411,0],[412,0.0384],[413,0.1343],[414,0.2591],[416,0.9151],[417,1],[418,0.8985],[419,0.6985],[420,0.5564],[424,0.3134],[425,0.2419],[426,0.1594],[427,0.0812],[428,0.0229],[429,0],[437,0],[438,0.0087],[439,0.0304],[440,0.0586],[441,0.2662],[442,0.6449],[443,0.9165],[444,1],[445,0.9555],[446,0.8576],[447,0.6461],[448,0.3457],[449,0.1369],[450,0.0428],[451,0],[482,0],[483,0.0842],[484,0.2696],[485,0.8604],[486,0.9165],[487,1],[488,0.8494],[489,0.5336],[490,0.2565],[491,0.1115],[492,0.0578],[493,0.0166],[494,0],[518,0],[519,0.0053],[520,0.019],[521,0.0385],[522,0.0608],[523,0.2649],[524,0.644],[525,0.9165],[526,1],[527,0.9555],[528,0.8577],[529,0.6712],[530,0.4103],[531,0.173],[532,0.0462],[533,0.0145],[534,0.0001],[576,0.0001]]
       }
       CTRL_L_eye_eyelidU.X {type=MovieSceneScriptingFloatChannel, default=0, const [0.071,577]}
       CTRL_L_eye_eyelidU.Y {
        type=MovieSceneScriptingFloatChannel
        default=0
        thinned_from=577
        keys [[0,0.057],[40,0.057],[41,-0.0345],[42,-0.0243],[43,0.0005],[44,0.0309],[45,0.0582],[46,0.0802],[47,0.0971],[48,0.1039],[105,0.1039],[106,0.0066],[107,0.0245],[108,0.0384],[109,0.0506],[110,0.0594],[111,0.0653],[112,0.068],[113,0.0669],[114,0.0642],[115,0.0609],[116,0.0582],[117,0.057],[276,0.0547],[281,0.0538],[293,0.0494],[296,0.049],[300,0.049],[325,0.0529],[401,0.0532],[402,0.053],[403,0.0524],[405,0.0505],[406,0.0422],[407,0.0244],[408,0.0043],[409,-0.0109],[410,-0.0208],[411,-0.0283],[412,-0.0314],[416,-0.0314],[417,-0.1163],[418,-0.1235],[419,-0.0968],[420,-0.0656],[424,-0.0305],[425,-0.0119],[426,0.0176],[427,0.045],[428,0.057],[443,0.057],[444,-0.0302],[445,-0.036],[446,-0.0232],[447,0.0034],[448,0.0264],[449,0.0412],[450,0.0525],[451,0.057],[486,0.057],[487,-0.0302],[488,-0.0355],[489,-0.0056],[490,0.0294],[491,0.0396],[492,0.0484],[493,0.0546],[494,0.057],[525,0.057],[526,-0.0302],[527,-0.0364],[528,-0.0243],[529,0.0006],[530,0.0215],[531,0.0346],[532,0.0459],[533,0.054],[534,0.057],[576,0.057]]
       }
       CTRL_L_eye_eyelidD.X {type=MovieSceneScriptingFloatChannel, default=0, const [0,577]}
       CTRL_L_eye_eyelidD.Y {
        type=MovieSceneScriptingFloatChannel
        default=0
        thinned_from=577
        keys [[0,0.0068],[40,0.0068],[41,-0.0905],[42,-0.0813],[43,-0.06],[44,-0.0356],[45,-0.0174],[46,-0.0057],[47,0.0032],[48,0.0068],[105,0.0068],[106,-0.0905],[107,-0.0668],[108,-0.0478],[109,-0.0316],[110,-0.0212],[112,-0.0083],[113,-0.0025],[114,0.0023],[115,0.0056],[116,0.0068],[172,0.0068],[176,0.008],[180,0.01],[181,0.0116],[182,0.0149],[183,0.0195],[184,0.025],[186,0.0368],[187,0.0421],[188,0.0465],[189,0.0494],[190,0.0505],[191,0.0496],[192,0.0472],[193,0.0435],[194,0.039],[195,0.0339],[197,0.0235],[198,0.0187],[199,0.0148],[200,0.012],[201,0.0106],[207,0.0078],[210,0.0069],[220,0.0068],[246,0.0084],[261,0.0101],[278,0.0162],[279,0.017],[280,0.0184],[281,0.0203],[282,0.0228],[283,0.0258],[284,0.0291],[285,0.0328],[286,0.0368],[292,0.0624],[293,0.0664],[294,0.0702],[295,0.0737],[296,0.0768],[297,0.0795],[298,0.0817],[299,0.0834],[300,0.0844],[301,0.0848],[302,0.0845],[303,0.0837],[304,0.0824],[305,0.0807],[306,0.0786],[307,0.0761],[308,0.0732],[309,0.0701],[310,0.0667],[312,0.0594],[319,0.0314],[321,0.024],[322,0.0206],[323,0.0175],[324,0.0146],[325,0.0121],[326,0.01],[327,0.0083],[328,0.007],[329,0.0062],[330,0.0059],[334,0.0059],[350,0.0067],[351,0.0065],[353,0.0059],[367,0.0063],[416,0.0067],[417,-0.079],[418,-0.0897],[419,-0.0633],[420,-0.0336],[422,-0.0204],[423,-0.0142],[424,-0.0085],[425,-0.0034],[426,0.0008],[427,0.004],[428,0.006],[429,0.0068],[443,0.0068],[444,-0.0804],[445,-0.0862],[446,-0.0735],[447,-0.0468],[448,-0.0238],[449,-0.0091],[450,0.0022],[451,0.0068],[486,0.0068],[487,-0.0804],[488,-0.0858],[489,-0.0559],[490,-0.0208],[491,-0.0107],[492,-0.0019],[493,0.0044],[494,0.0068],[525,0.0068],[526,-0.0804],[527,-0.0866],[528,-0.0746],[529,-0.0496],[530,-0.0287],[531,-0.0157],[532,-0.0043],[533,0.0037],[534,0.0068],[576,0.0068]]
       }
       CTRL_L_brow.X {
        type=MovieSceneScriptingFloatChannel
        default=0
        thinned_from=577
        keys [[0,0.0122],[41,0.0122],[42,0.012],[43,0.0113],[47,0.0069],[48,0.0059],[49,0.0053],[51,0.0045],[53,0.0042],[59,0.0041],[74,0.0037],[90,0.0035],[101,0.0034],[106,0.005],[113,0.0087],[117,0.0095],[178,0.0095],[189,0.0115],[288,0.0115],[414,0.0112],[576,0.0122]]
       }
       CTRL_L_brow.Y {
        type=MovieSceneScriptingFloatChannel
        default=0
        thinned_from=577
        keys [[0,-0.116],[40,-0.116],[41,-0.1157],[42,-0.1138],[43,-0.1091],[44,-0.1024],[45,-0.0945],[46,-0.0862],[47,-0.0782],[48,-0.0713],[49,-0.0664],[50,-0.063],[51,-0.0602],[52,-0.0581],[53,-0.0565],[54,-0.0553],[55,-0.0545],[56,-0.054],[58,-0.0535],[101,-0.0519],[103,-0.0529],[105,-0.0546],[106,-0.0556],[107,-0.057],[109,-0.0602],[112,-0.0654],[113,-0.067],[114,-0.0683],[116,-0.0704],[121,-0.0738],[123,-0.0745],[169,-0.0747],[171,-0.0735],[173,-0.0716],[178,-0.0649],[181,-0.0602],[198,-0.0296],[201,-0.0249],[202,-0.0235],[203,-0.0224],[204,-0.0219],[240,-0.0217],[243,-0.0228],[246,-0.0247],[250,-0.0282],[266,-0.0468],[268,-0.0495],[274,-0.0587],[276,-0.0613],[284,-0.0692],[292,-0.0747],[296,-0.0768],[297,-0.0768],[298,-0.0756],[299,-0.0739],[300,-0.073],[327,-0.0787],[345,-0.0788],[346,-0.0792],[347,-0.0804],[348,-0.0819],[349,-0.0832],[350,-0.0838],[353,-0.0796],[354,-0.0789],[372,-0.0789],[374,-0.0798],[379,-0.0832],[380,-0.0836],[381,-0.0838],[404,-0.0838],[405,-0.0845],[406,-0.0874],[407,-0.092],[408,-0.0978],[410,-0.1108],[411,-0.1169],[412,-0.1222],[413,-0.1259],[414,-0.1277],[415,-0.1274],[416,-0.1258],[417,-0.1231],[418,-0.1195],[419,-0.1152],[420,-0.1104],[423,-0.0947],[424,-0.0898],[425,-0.0852],[426,-0.0812],[427,-0.0781],[428,-0.0759],[429,-0.0749],[487,-0.0749],[488,-0.0754],[489,-0.0762],[490,-0.0772],[491,-0.0785],[492,-0.08],[494,-0.0837],[496,-0.0879],[503,-0.1041],[505,-0.1082],[507,-0.1117],[508,-0.1131],[509,-0.1142],[510,-0.1151],[511,-0.1157],[512,-0.116],[576,-0.116]]
       }
       CTRL_L_brow_inner {
        type=MovieSceneScriptingFloatChannel
        default=0
        thinned_from=577
        keys [[0,-0.0845],[40,-0.0845],[41,-0.0842],[42,-0.0819],[43,-0.0763],[44,-0.0684],[45,-0.0592],[46,-0.0496],[47,-0.0404],[48,-0.0326],[49,-0.027],[50,-0.0235],[51,-0.021],[52,-0.0193],[53,-0.0184],[54,-0.0178],[71,-0.0151],[100,-0.0143],[101,-0.0149],[102,-0.0167],[103,-0.0196],[104,-0.0231],[106,-0.0311],[107,-0.0358],[108,-0.0412],[111,-0.0589],[112,-0.0642],[113,-0.0686],[114,-0.0719],[115,-0.0741],[116,-0.0756],[117,-0.0767],[118,-0.0773],[119,-0.0777],[168,-0.0778],[169,-0.0777],[170,-0.0773],[171,-0.0768],[173,-0.075],[175,-0.0726],[177,-0.0696],[178,-0.0679],[179,-0.0666],[180,-0.0662],[181,-0.0665],[182,-0.0674],[185,-0.0711],[186,-0.0721],[187,-0.0728],[188,-0.0728],[189,-0.0721],[190,-0.0704],[193,-0.0632],[197,-0.053],[199,-0.0483],[201,-0.0443],[202,-0.0426],[203,-0.0414],[204,-0.0407],[205,-0.0405],[239,-0.0404],[246,-0.0392],[273,-0.0284],[274,-0.0282],[279,-0.0282],[281,-0.0286],[293,-0.0341],[296,-0.035],[297,-0.0349],[298,-0.0335],[299,-0.0317],[300,-0.0306],[345,-0.0323],[346,-0.0327],[347,-0.0339],[348,-0.0354],[349,-0.0367],[350,-0.0373],[353,-0.0331],[354,-0.0324],[372,-0.0324],[374,-0.0333],[379,-0.0367],[380,-0.0371],[381,-0.0373],[404,-0.0373],[405,-0.0379],[406,-0.0404],[407,-0.0445],[408,-0.0495],[410,-0.0609],[411,-0.0662],[412,-0.0708],[413,-0.0741],[414,-0.0757],[415,-0.0754],[416,-0.074],[417,-0.0717],[418,-0.0687],[419,-0.065],[420,-0.0609],[423,-0.0476],[424,-0.0433],[425,-0.0394],[426,-0.0361],[427,-0.0334],[428,-0.0315],[429,-0.0307],[487,-0.0307],[488,-0.0313],[489,-0.0323],[490,-0.0337],[491,-0.0354],[492,-0.0374],[493,-0.0397],[495,-0.0448],[498,-0.0537],[502,-0.066],[504,-0.0717],[505,-0.0743],[506,-0.0766],[507,-0.0788],[508,-0.0806],[509,-0.0822],[510,-0.0833],[511,-0.0841],[512,-0.0844],[576,-0.0844]]
       }
       CTRL_L_brow_outer {
        type=MovieSceneScriptingFloatChannel
        default=0
        thinned_from=577
        keys [[0,-0.069],[39,-0.069],[40,-0.0687],[44,-0.0687],[93,-0.0698],[100,-0.07],[102,-0.0705],[105,-0.0722],[107,-0.0737],[112,-0.0787],[113,-0.0795],[115,-0.0805],[118,-0.081],[132,-0.081],[178,-0.0811],[181,-0.0818],[187,-0.0843],[189,-0.0848],[241,-0.0848],[248,-0.083],[268,-0.0738],[275,-0.0698],[281,-0.0681],[286,-0.068],[296,-0.0696],[405,-0.0706],[406,-0.0703],[407,-0.0698],[408,-0.069],[409,-0.068],[410,-0.0667],[411,-0.0653],[413,-0.0618],[414,-0.0598],[415,-0.0575],[416,-0.0547],[417,-0.0515],[418,-0.048],[422,-0.033],[423,-0.0294],[424,-0.0261],[425,-0.0231],[426,-0.0206],[427,-0.0187],[428,-0.0173],[429,-0.0167],[486,-0.0167],[487,-0.0168],[488,-0.0174],[489,-0.0183],[490,-0.0197],[491,-0.0213],[492,-0.0233],[493,-0.0255],[495,-0.0305],[498,-0.0391],[502,-0.051],[504,-0.0565],[505,-0.0591],[506,-0.0614],[507,-0.0635],[508,-0.0653],[509,-0.0667],[510,-0.0679],[511,-0.0686],[512,-0.069],[576,-0.069]]
       }
       CTRL_C_nose.X {type=MovieSceneScriptingFloatChannel, default=0, const [0,577]}
       CTRL_C_nose.Y {type=MovieSceneScriptingFloatChannel, default=0, const [0,577]}
       CTRL_L_nose.X {type=MovieSceneScriptingFloatChannel, default=0, thinned_from=577, keys [[0,0],[103,0],[106,0.0005],[111,0.0004],[112,0.0002],[576,0]]}
       CTRL_L_nose.Y {
        type=MovieSceneScriptingFloatChannel
        default=0
        thinned_from=577
        keys [[0,0.0584],[103,0.0584],[106,0.0592],[108,0.0593],[114,0.0583],[293,0.0583],[329,0.0649],[336,0.0674],[337,0.068],[340,0.0704],[341,0.071],[343,0.0716],[352,0.0718],[404,0.0732],[405,0.073],[406,0.0724],[407,0.0715],[409,0.0689],[412,0.0647],[413,0.0636],[414,0.0627],[419,0.0602],[425,0.0586],[576,0.0583]]
       }
       noseWrinkleUpper {type=MovieSceneScriptingFloatChannel, default=0, const [0.75,577]}
       CTRL_R_nose.X {type=MovieSceneScriptingFloatChannel, default=0, const [0,577]}
       CTRL_R_nose.Y {
        type=MovieSceneScriptingFloatChannel
        default=0
        thinned_from=577
        keys [[0,-0.0041],[294,-0.0041],[332,0.0005],[347,0.0038],[368,0.0045],[405,0.0049],[407,0.0039],[412,-0.0002],[414,-0.0015],[422,-0.0036],[576,-0.0041]]
       }
       noseWrinkleUpper_2 {type=MovieSceneScriptingFloatChannel, default=0, const [0.75,577]}
       CTRL_L_eye_squintInner {
        type=MovieSceneScriptingFloatChannel
        default=0
        thinned_from=577
        keys [[0,0],[106,0],[107,0.0138],[108,0.0231],[109,0.0307],[110,0.035],[111,0.0363],[112,0.0336],[113,0.0269],[115,0.0094],[116,0.0027],[117,0],[182,0],[183,0.0129],[184,0.0275],[185,0.0388],[186,0.0433],[187,0.0415],[188,0.037],[189,0.0312],[190,0.0186],[191,0.0066],[192,0.0021],[193,0],[247,0],[261,0.0008],[278,0.0026],[280,0.0035],[303,0.0187],[319,0.0259],[326,0.0278],[400,0.028],[402,0.0281],[405,0.0287],[406,0.0302],[407,0.0337],[409,0.0426],[410,0.046],[411,0.0474],[412,0.0453],[413,0.04],[414,0.0332],[415,0.0175],[416,0.003],[417,0],[576,0]]
       }
       CTRL_L_eye_cheekRaise {
        type=MovieSceneScriptingFloatChannel
        default=0
        thinned_from=577
        keys [[0,0],[40,0],[41,0.0383],[42,0.0593],[43,0.1055],[44,0.1558],[45,0.1968],[46,0.2247],[47,0.2432],[48,0.25],[49,0.25],[50,0.2428],[51,0.2325],[52,0.2203],[53,0.2074],[54,0.1951],[55,0.1844],[56,0.1773],[57,0.1746],[101,0.1746],[102,0.1475],[103,0.0878],[104,0.0279],[105,0],[106,0.0112],[107,0],[108,0],[109,0.0032],[110,0.0147],[111,0.0247],[112,0.0278],[113,0.0189],[114,0.0051],[115,0],[240,0],[244,0.0013],[248,0.0037],[255,0.0097],[266,0.0206],[268,0.0232],[274,0.0323],[277,0.0361],[279,0.0381],[280,0.0389],[281,0.0405],[282,0.0439],[283,0.0487],[284,0.0548],[285,0.0616],[286,0.0691],[287,0.0773],[288,0.0861],[289,0.0953],[293,0.1331],[294,0.1421],[295,0.1505],[296,0.1583],[297,0.1653],[298,0.1714],[299,0.1759],[300,0.1788],[301,0.1798],[302,0.1792],[303,0.1777],[304,0.1751],[305,0.1718],[306,0.1676],[307,0.1627],[308,0.1572],[309,0.1511],[310,0.1446],[311,0.1376],[312,0.1302],[314,0.1149],[317,0.0913],[318,0.0836],[319,0.0762],[320,0.069],[321,0.0622],[322,0.0557],[323,0.0498],[324,0.0445],[325,0.0398],[326,0.0357],[327,0.0325],[328,0.0301],[329,0.0286],[330,0.0281],[404,0.0281],[405,0.0285],[406,0.03],[407,0.0324],[408,0.0354],[410,0.0417],[411,0.0445],[412,0.0465],[413,0.0474],[414,0.0364],[415,0.0064],[416,0.0319],[417,0.0589],[418,0.0666],[419,0.0634],[420,0.0583],[421,0.0517],[422,0.044],[424,0.0273],[425,0.0193],[426,0.0121],[427,0.0062],[428,0.002],[429,0.0001],[443,0.0001],[444,0.0319],[445,0.0566],[446,0.1273],[447,0.2076],[448,0.2367],[449,0.2477],[450,0.25],[451,0.2491],[452,0.2455],[453,0.24],[454,0.2332],[456,0.2188],[457,0.2125],[458,0.2079],[459,0.2056],[483,0.2054],[484,0.1813],[485,0.1048],[486,0.0178],[487,0.0319],[488,0.0566],[489,0.1273],[490,0.2076],[491,0.2367],[492,0.2477],[493,0.25],[494,0.2491],[495,0.2455],[496,0.24],[497,0.2332],[499,0.2188],[500,0.2125],[501,0.2079],[502,0.2056],[522,0.2054],[523,0.1813],[524,0.1048],[525,0.0178],[526,0.0319],[527,0.0586],[528,0.1056],[529,0.1623],[530,0.2099],[531,0.2361],[532,0.2473],[533,0.25],[534,0.2497],[535,0.2485],[536,0.2464],[537,0.2431],[538,0.2387],[539,0.2331],[540,0.2262],[541,0.2178],[542,0.208],[543,0.1962],[544,0.1817],[545,0.1648],[546,0.1464],[547,0.1269],[549,0.087],[550,0.0678],[551,0.0499],[552,0.0338],[553,0.0202],[554,0.0097],[555,0.0027],[556,0],[576,0]]
       }
       CTRL_C_mouth.X {
        type=MovieSceneScriptingFloatChannel
        default=0
        thinned_from=577
        keys [[0,0],[214,0],[218,-0.0008],[227,-0.0044],[229,-0.0048],[262,-0.005],[264,-0.0048],[274,-0.0008],[278,-0.0002],[576,0]]
       }
       CTRL_C_mouth.Y {
        type=MovieSceneScriptingFloatChannel
        default=0
        thinned_from=577
        keys [[0,0.0518],[41,0.0518],[42,0.0535],[43,0.0545],[44,0.0542],[45,0.0528],[46,0.0506],[47,0.0479],[49,0.0422],[50,0.0425],[105,0.0425],[107,0.039],[110,0.033],[111,0.0312],[112,0.0298],[113,0.0288],[114,0.0284],[211,0.0284],[214,0.0281],[215,0.0277],[216,0.0268],[217,0.0255],[218,0.0239],[219,0.0219],[221,0.0174],[224,0.0103],[226,0.0061],[227,0.0045],[228,0.002],[229,-0.0026],[230,-0.0078],[231,-0.0122],[232,-0.0162],[233,-0.0195],[234,-0.0221],[235,-0.0238],[236,-0.0246],[239,-0.0255],[258,-0.0259],[259,-0.0256],[261,-0.0244],[264,-0.0218],[265,-0.0177],[269,0.0043],[270,0.0096],[271,0.0146],[272,0.0193],[273,0.0238],[274,0.0279],[275,0.0316],[276,0.0341],[277,0.0352],[278,0.0357],[279,0.0379],[280,0.043],[281,0.0503],[282,0.059],[283,0.0636],[284,0.064],[287,0.0659],[289,0.0661],[297,0.0642],[302,0.064],[315,0.0627],[316,0.0623],[318,0.0611],[326,0.061],[341,0.0631],[349,0.0655],[374,0.0688],[405,0.0694],[406,0.0686],[407,0.0675],[408,0.0661],[412,0.0595],[413,0.0581],[414,0.0571],[418,0.0546],[423,0.0526],[427,0.0519],[576,0.0518]]
       }
       CTRL_L_mouth_lipU.X {type=MovieSceneScriptingFloatChannel, default=0, const [0.118,577]}
       CTRL_L_mouth_lipU.Y {
        type=MovieSceneScriptingFloatChannel
        default=0
        thinned_from=577
        keys [[0,0.1026],[42,0.1026],[43,0.1031],[44,0.1043],[47,0.1094],[48,0.1106],[49,0.1111],[50,0.1103],[51,0.1081],[52,0.105],[53,0.1013],[55,0.0937],[56,0.0905],[57,0.0883],[58,0.0875],[102,0.0875],[103,0.0889],[105,0.0954],[106,0.0954],[107,0.0937],[111,0.0847],[112,0.0827],[113,0.0811],[114,0.08],[119,0.0765],[120,0.076],[213,0.0759],[215,0.076],[227,0.0807],[230,0.0813],[259,0.0814],[263,0.0807],[265,0.0809],[267,0.0819],[269,0.0835],[271,0.0857],[272,0.0872],[275,0.0927],[276,0.0948],[277,0.0972],[278,0.0999],[279,0.103],[280,0.1056],[281,0.1061],[282,0.1057],[285,0.1032],[287,0.1013],[288,0.1007],[292,0.0996],[295,0.0995],[334,0.1205],[337,0.1227],[340,0.1261],[341,0.1271],[343,0.1282],[362,0.1319],[404,0.1327],[405,0.1325],[406,0.1313],[407,0.1294],[408,0.127],[411,0.1184],[412,0.1157],[413,0.1133],[414,0.1116],[417,0.1083],[420,0.1058],[423,0.104],[426,0.103],[428,0.1027],[576,0.1027]]
       }
       mouthPress {type=MovieSceneScriptingFloatChannel, default=0, const [0,577]}
       lipsTogether {type=MovieSceneScriptingFloatChannel, default=0, const [0,577]}
       CTRL_L_mouth_lipD.X {type=MovieSceneScriptingFloatChannel, default=0, const [0.0654,577]}
       CTRL_L_mouth_lipD.Y {
        type=MovieSceneScriptingFloatChannel
        default=0
        thinned_from=577
        keys [[0,-0.1872],[41,-0.1872],[42,-0.1862],[43,-0.1872],[44,-0.1913],[45,-0.1972],[46,-0.2039],[47,-0.21],[48,-0.2144],[49,-0.216],[50,-0.2146],[51,-0.2112],[52,-0.2064],[55,-0.1888],[56,-0.1838],[57,-0.1804],[58,-0.1792],[102,-0.1792],[103,-0.1804],[105,-0.1857],[106,-0.1867],[109,-0.1851],[114,-0.181],[120,-0.1777],[212,-0.1773],[214,-0.1768],[215,-0.1758],[216,-0.1738],[217,-0.1708],[218,-0.1671],[219,-0.1627],[220,-0.1578],[224,-0.1366],[225,-0.1316],[226,-0.1271],[227,-0.1232],[228,-0.1201],[229,-0.1179],[230,-0.1167],[233,-0.1155],[242,-0.115],[244,-0.113],[245,-0.1127],[250,-0.1127],[257,-0.1138],[258,-0.1142],[259,-0.1148],[264,-0.1186],[265,-0.1225],[270,-0.1494],[273,-0.1648],[274,-0.1696],[275,-0.1741],[276,-0.1776],[278,-0.1825],[280,-0.1862],[281,-0.1865],[288,-0.183],[294,-0.1815],[347,-0.1929],[405,-0.1936],[414,-0.1891],[576,-0.1872]]
       }
       mouthPress_2 {type=MovieSceneScriptingFloatChannel, default=0, const [0,577]}
       lipsTogether_2 {type=MovieSceneScriptingFloatChannel, default=0, const [0,577]}
       CTRL_L_mouth_smileDepress.X {
        type=MovieSceneScriptingFloatChannel
        default=0
        thinned_from=577
        keys [[0,0.0452],[42,0.0452],[43,0.0439],[44,0.0406],[45,0.0362],[46,0.0314],[47,0.027],[48,0.0238],[49,0.0226],[212,0.0226],[214,0.0227],[215,0.023],[216,0.0235],[217,0.0242],[219,0.0263],[225,0.0342],[227,0.0363],[228,0.0371],[229,0.0377],[232,0.0382],[258,0.0383],[264,0.0397],[265,0.0409],[270,0.0488],[273,0.0528],[274,0.0538],[275,0.0546],[276,0.055],[278,0.055],[281,0.0539],[282,0.0533],[283,0.0523],[284,0.0508],[285,0.049],[286,0.0469],[287,0.0445],[288,0.0417],[292,0.0294],[293,0.0265],[294,0.0239],[295,0.0215],[296,0.0195],[297,0.0179],[298,0.0169],[299,0.0165],[301,0.0166],[303,0.0176],[305,0.0192],[307,0.0214],[310,0.0256],[319,0.0403],[322,0.0446],[324,0.047],[326,0.0488],[328,0.0499],[329,0.0502],[338,0.0502],[342,0.0503],[405,0.0512],[407,0.0507],[413,0.0474],[423,0.0455],[576,0.0452]]
       }
       CTRL_L_mouth_smileDepress.Y {
        type=MovieSceneScriptingFloatChannel
        default=0
        thinned_from=577
        keys [[0,0.0692],[42,0.0692],[43,0.068],[44,0.0653],[45,0.0615],[46,0.0574],[47,0.0537],[48,0.0509],[49,0.0499],[214,0.0499],[216,0.0507],[218,0.0521],[225,0.0598],[227,0.0616],[229,0.0628],[231,0.0631],[233,0.0631],[258,0.0633],[264,0.0645],[265,0.0655],[270,0.0723],[273,0.0757],[274,0.0766],[275,0.0772],[276,0.0776],[278,0.0776],[280,0.0769],[281,0.0769],[283,0.0783],[285,0.0802],[288,0.0836],[294,0.0893],[296,0.0906],[298,0.0914],[299,0.0917],[301,0.0917],[308,0.0894],[309,0.0887],[310,0.0873],[311,0.0852],[321,0.0753],[325,0.0723],[328,0.0708],[333,0.0701],[576,0.0692]]
       }
       CTRL_L_mouth_corner.X {
        type=MovieSceneScriptingFloatChannel
        default=0
        thinned_from=577
        keys [[0,-0.0727],[102,-0.0727],[103,-0.0708],[105,-0.0622],[106,-0.0622],[107,-0.0645],[108,-0.0673],[111,-0.0766],[112,-0.0793],[113,-0.0815],[114,-0.0829],[119,-0.0876],[120,-0.0883],[121,-0.0885],[187,-0.0885],[188,-0.0887],[189,-0.0875],[190,-0.0816],[191,-0.0728],[192,-0.0659],[193,-0.0607],[194,-0.0578],[195,-0.0569],[197,-0.0572],[198,-0.0575],[199,-0.058],[200,-0.0589],[201,-0.0602],[202,-0.062],[203,-0.0643],[204,-0.0673],[205,-0.0735],[206,-0.0842],[207,-0.0885],[211,-0.0886],[214,-0.0891],[215,-0.09],[216,-0.0919],[217,-0.0948],[218,-0.0984],[219,-0.1027],[220,-0.1074],[224,-0.128],[225,-0.1328],[226,-0.1372],[227,-0.141],[228,-0.144],[229,-0.1462],[230,-0.1473],[234,-0.1487],[258,-0.1489],[260,-0.1483],[264,-0.1456],[265,-0.142],[270,-0.1178],[272,-0.1088],[273,-0.1045],[274,-0.1004],[275,-0.0967],[276,-0.094],[277,-0.0923],[278,-0.0908],[280,-0.0887],[282,-0.0884],[294,-0.0881],[313,-0.0784],[314,-0.0789],[316,-0.081],[317,-0.0817],[318,-0.0819],[319,-0.0809],[320,-0.0786],[321,-0.076],[322,-0.0743],[327,-0.073],[411,-0.0726],[576,-0.0727]]
       }
       CTRL_L_mouth_corner.Y {
        type=MovieSceneScriptingFloatChannel
        default=0
        thinned_from=577
        keys [[0,-0.2178],[215,-0.2181],[218,-0.22],[227,-0.2295],[230,-0.2309],[264,-0.2306],[276,-0.2169],[277,-0.2154],[278,-0.213],[279,-0.2098],[280,-0.2083],[281,-0.2054],[282,-0.2032],[283,-0.2015],[284,-0.2011],[285,-0.2016],[287,-0.2038],[293,-0.2153],[294,-0.2168],[296,-0.218],[576,-0.2178]]
       }
       lipsBlow {type=MovieSceneScriptingFloatChannel, default=0, const [0,577]}
       mouthDimple {type=MovieSceneScriptingFloatChannel, default=0, const [0,577]}
       mouthStretch {type=MovieSceneScriptingFloatChannel, default=0, const [0.0299,577]}
       mouthPurseU {type=MovieSceneScriptingFloatChannel, default=0, const [0,577]}
       mouthPurseD {type=MovieSceneScriptingFloatChannel, default=0, const [0,577]}
       mouthFunnelU {type=MovieSceneScriptingFloatChannel, default=0, const [0,577]}
       mouthFunnelD {type=MovieSceneScriptingFloatChannel, default=0, const [0,577]}
       mouthCornerSharpnessU {type=MovieSceneScriptingFloatChannel, default=0, const [0,577]}
       mouthCornerSharpnessD {type=MovieSceneScriptingFloatChannel, default=0, const [0,577]}
       CTRL_R_mouth_lipU.X {type=MovieSceneScriptingFloatChannel, default=0, const [0,577]}
       CTRL_R_mouth_lipU.Y {
        type=MovieSceneScriptingFloatChannel
        default=0
        thinned_from=577
        keys [[0,-0.017],[42,-0.017],[43,-0.0172],[45,-0.0184],[47,-0.02],[48,-0.0205],[49,-0.0207],[211,-0.0207],[214,-0.0203],[215,-0.0197],[216,-0.0183],[217,-0.0164],[218,-0.0139],[219,-0.011],[220,-0.0078],[224,0.0063],[225,0.0096],[226,0.0126],[227,0.0152],[228,0.0173],[229,0.0188],[230,0.0195],[234,0.0205],[258,0.0206],[260,0.02],[264,0.0179],[265,0.0157],[269,0.0039],[271,-0.0014],[273,-0.0062],[275,-0.0102],[276,-0.0116],[278,-0.0133],[280,-0.0143],[282,-0.0141],[287,-0.0126],[290,-0.0126],[318,-0.0161],[576,-0.0169]]
       }
       mouthPress_3 {type=MovieSceneScriptingFloatChannel, default=0, const [0,577]}
       lipsTogether_3 {type=MovieSceneScriptingFloatChannel, default=0, const [0,577]}
       CTRL_R_mouth_lipD.X {
        type=MovieSceneScriptingFloatChannel
        default=0
        thinned_from=577
        keys [[0,-0.0526],[211,-0.0526],[214,-0.0522],[215,-0.0518],[216,-0.0508],[217,-0.0494],[218,-0.0476],[219,-0.0454],[221,-0.0404],[224,-0.0325],[225,-0.0301],[226,-0.0279],[227,-0.0259],[228,-0.0244],[229,-0.0233],[230,-0.0227],[232,-0.0223],[233,-0.0208],[234,-0.0177],[235,-0.0158],[236,-0.0149],[239,-0.0138],[241,-0.0135],[258,-0.0134],[264,-0.0151],[265,-0.0169],[270,-0.0292],[273,-0.0359],[275,-0.0399],[276,-0.0412],[278,-0.043],[279,-0.0449],[280,-0.0485],[281,-0.0523],[576,-0.0525]]
       }
       CTRL_R_mouth_lipD.Y {
        type=MovieSceneScriptingFloatChannel
        default=0
        thinned_from=577
        keys [[0,-0.028],[208,-0.028],[213,-0.0274],[214,-0.0271],[215,-0.0261],[216,-0.0239],[217,-0.0207],[218,-0.0167],[219,-0.0119],[220,-0.0067],[224,0.0162],[225,0.0215],[226,0.0263],[227,0.0305],[228,0.0339],[229,0.0363],[230,0.0376],[232,0.0386],[235,0.0392],[258,0.0394],[260,0.0386],[264,0.0357],[265,0.0316],[269,0.01],[271,-0.0003],[273,-0.0101],[274,-0.0146],[275,-0.0187],[276,-0.0218],[277,-0.0237],[278,-0.0253],[279,-0.0265],[280,-0.0275],[281,-0.0278],[285,-0.0278],[295,-0.0281],[331,-0.0329],[347,-0.0371],[405,-0.0384],[407,-0.0372],[412,-0.0325],[414,-0.0311],[421,-0.0288],[425,-0.0282],[576,-0.028]]
       }
       mouthPress_2_2 {type=MovieSceneScriptingFloatChannel, default=0, const [0,577]}
       lipsTogether_2_2 {type=MovieSceneScriptingFloatChannel, default=0, const [0,577]}
       CTRL_R_mouth_smileDepress.X {type=MovieSceneScriptingFloatChannel, default=0, const [0,577]}
       CTRL_R_mouth_smileDepress.Y {
        type=MovieSceneScriptingFloatChannel
        default=0
        thinned_from=577
        keys [[0,0],[264,0],[270,0.0009],[277,0.0011],[280,0.0005],[282,0.0005],[287,0.0012],[291,0.0012],[316,0.0003],[319,0.0007],[320,0.0007],[323,0.0001],[576,0.0001]]
       }
       CTRL_R_mouth_corner.X {
        type=MovieSceneScriptingFloatChannel
        default=0
        thinned_from=577
        keys [[0,-0.25],[100,-0.25],[101,-0.2328],[102,-0.2151],[103,-0.1996],[104,-0.1906],[105,-0.1892],[106,-0.1979],[107,-0.2042],[110,-0.2259],[111,-0.2322],[112,-0.2372],[113,-0.2402],[114,-0.2408],[119,-0.2366],[121,-0.2358],[216,-0.2361],[229,-0.2415],[259,-0.2416],[264,-0.2397],[265,-0.2381],[270,-0.2275],[273,-0.2231],[274,-0.2222],[275,-0.2219],[276,-0.2223],[277,-0.2233],[279,-0.2269],[280,-0.2259],[281,-0.2192],[282,-0.208],[283,-0.1934],[284,-0.1765],[285,-0.1577],[286,-0.1374],[287,-0.1163],[288,-0.0957],[289,-0.0757],[290,-0.0565],[291,-0.0383],[292,-0.0216],[293,-0.0069],[294,0.0055],[295,0.0147],[296,0.0204],[297,0.0233],[298,0.0137],[299,-0.0165],[300,-0.0619],[301,-0.1138],[302,-0.1642],[303,-0.205],[304,-0.2272],[305,-0.2288],[306,-0.2176],[307,-0.1993],[308,-0.1767],[309,-0.1526],[310,-0.1302],[311,-0.1123],[312,-0.1023],[313,-0.1005],[314,-0.102],[315,-0.103],[316,-0.1047],[317,-0.1082],[318,-0.1121],[319,-0.113],[320,-0.1127],[321,-0.1147],[322,-0.1306],[323,-0.1649],[324,-0.2049],[325,-0.2369],[326,-0.2495],[576,-0.2499]]
       }
       CTRL_R_mouth_corner.Y {
        type=MovieSceneScriptingFloatChannel
        default=0
        thinned_from=577
        keys [[0,-0.0755],[42,-0.0755],[43,-0.076],[44,-0.0774],[47,-0.0833],[48,-0.0847],[49,-0.0853],[210,-0.0853],[214,-0.086],[215,-0.087],[216,-0.0892],[217,-0.0925],[218,-0.0966],[219,-0.1015],[220,-0.1069],[224,-0.1304],[225,-0.1359],[226,-0.1409],[227,-0.1452],[228,-0.1487],[229,-0.1512],[230,-0.1525],[233,-0.1538],[258,-0.1543],[260,-0.1534],[264,-0.1496],[265,-0.1445],[269,-0.1173],[270,-0.1107],[272,-0.0983],[273,-0.0923],[274,-0.0868],[275,-0.0818],[276,-0.0781],[277,-0.076],[278,-0.0742],[280,-0.072],[287,-0.0707],[295,-0.0707],[324,-0.0648],[331,-0.0621],[345,-0.0539],[349,-0.0524],[360,-0.0507],[395,-0.0499],[396,-0.0497],[404,-0.0454],[405,-0.0452],[406,-0.0458],[407,-0.0471],[408,-0.0488],[412,-0.0575],[413,-0.0593],[414,-0.0606],[418,-0.064],[421,-0.0659],[425,-0.0675],[428,-0.068],[435,-0.068],[436,-0.0685],[437,-0.0715],[438,-0.0751],[439,-0.0755],[576,-0.0755]]
       }
       lipsBlow_2 {type=MovieSceneScriptingFloatChannel, default=0, const [0,577]}
       mouthDimple_2 {type=MovieSceneScriptingFloatChannel, default=0, const [0.1645,577]}
       mouthStretch_2 {type=MovieSceneScriptingFloatChannel, default=0, const [0,577]}
       mouthPurseU_2 {type=MovieSceneScriptingFloatChannel, default=0, const [0,577]}
       mouthPurseD_2 {type=MovieSceneScriptingFloatChannel, default=0, const [0,577]}
       mouthFunnelU_2 {type=MovieSceneScriptingFloatChannel, default=0, const [0,577]}
       mouthFunnelD_2 {type=MovieSceneScriptingFloatChannel, default=0, const [0,577]}
       mouthCornerSharpnessU_2 {type=MovieSceneScriptingFloatChannel, default=0, const [0,577]}
       mouthCornerSharpnessD_2 {type=MovieSceneScriptingFloatChannel, default=0, const [0,577]}
       CTRL_L_mouth_suckBlow {type=MovieSceneScriptingFloatChannel, default=0, const [0,577]}
       CTRL_C_jaw.Location.X {
        type=MovieSceneScriptingFloatChannel
        default=0
        thinned_from=577
        keys [[0,0],[215,0],[219,0.0012],[227,0.0049],[230,0.0055],[261,0.0056],[264,0.0054],[275,0.0008],[279,0.0001],[576,0]]
       }
       CTRL_C_jaw.Location.Y {
        type=MovieSceneScriptingFloatChannel
        default=0
        thinned_from=577
        keys [[0,0.0317],[42,0.0317],[43,0.031],[44,0.0294],[45,0.0271],[48,0.0184],[49,0.0158],[214,0.0158],[215,0.0156],[216,0.0152],[218,0.0136],[221,0.0102],[225,0.0052],[227,0.0032],[228,0.0024],[229,0.0019],[231,0.0015],[258,0.0012],[261,0.002],[264,0.0034],[265,0.0057],[270,0.0207],[272,0.0263],[274,0.0313],[275,0.0334],[276,0.035],[278,0.0365],[280,0.0373],[291,0.0379],[294,0.0375],[345,0.0149],[375,0.0092],[404,0.0086],[405,0.0088],[406,0.0097],[407,0.0111],[408,0.013],[412,0.0217],[413,0.0235],[414,0.0248],[417,0.0273],[421,0.0297],[425,0.0312],[427,0.0315],[576,0.0317]]
       }
       CTRL_C_jaw.Location.Z {
        type=MovieSceneScriptingFloatChannel
        default=0
        thinned_from=577
        keys [[0,0],[290,0],[295,-0.0004],[326,-0.0082],[346,-0.0161],[375,-0.0211],[404,-0.0216],[405,-0.0215],[406,-0.0206],[407,-0.0193],[408,-0.0175],[412,-0.0094],[413,-0.0077],[414,-0.0065],[418,-0.0034],[422,-0.0014],[426,-0.0003],[576,0]]
       }
       jawOpenExtreme {type=MovieSceneScriptingFloatChannel, default=0, const [0,577]}
       chinRaiseUL {
        type=MovieSceneScriptingFloatChannel
        default=0
        thinned_from=577
        keys [[0,0],[229,0],[231,0.0048],[233,0.0087],[234,0.0101],[235,0.0109],[237,0.0117],[258,0.0121],[261,0.0113],[264,0.0099],[265,0.0076],[267,0.0016],[268,0],[576,0]]
       }
       chinRaiseUR {
        type=MovieSceneScriptingFloatChannel
        default=0
        thinned_from=577
        keys [[0,0],[229,0],[231,0.0048],[233,0.0087],[234,0.0101],[235,0.0109],[237,0.0117],[258,0.0121],[261,0.0113],[264,0.0099],[265,0.0076],[267,0.0016],[268,0],[576,0]]
       }
       chinRaiseDL {
        type=MovieSceneScriptingFloatChannel
        default=0
        thinned_from=577
        keys [[0,0],[229,0],[231,0.0048],[233,0.0087],[234,0.0101],[235,0.0109],[237,0.0117],[258,0.0121],[261,0.0113],[264,0.0099],[265,0.0076],[267,0.0016],[268,0],[576,0]]
       }
       chinRaiseDR {
        type=MovieSceneScriptingFloatChannel
        default=0
        thinned_from=577
        keys [[0,0],[229,0],[231,0.0048],[233,0.0087],[234,0.0101],[235,0.0109],[237,0.0117],[258,0.0121],[261,0.0113],[264,0.0099],[265,0.0076],[267,0.0016],[268,0],[576,0]]
       }
       CTRL_R_brow.X {type=MovieSceneScriptingFloatChannel, default=0, const [0,577]}
       CTRL_R_brow.Y {
        type=MovieSceneScriptingFloatChannel
        default=0
        thinned_from=577
        keys [[0,-0.1559],[40,-0.1559],[41,-0.1561],[42,-0.1556],[43,-0.1532],[44,-0.1496],[45,-0.1452],[47,-0.1358],[48,-0.1315],[49,-0.1276],[50,-0.1244],[51,-0.1219],[52,-0.1203],[55,-0.1171],[57,-0.1156],[59,-0.115],[103,-0.1141],[106,-0.1147],[113,-0.1175],[169,-0.1175],[171,-0.1165],[173,-0.1147],[175,-0.1123],[178,-0.1077],[180,-0.1041],[183,-0.0975],[188,-0.086],[190,-0.0818],[196,-0.0707],[200,-0.0643],[202,-0.0614],[203,-0.0604],[204,-0.0598],[240,-0.0596],[243,-0.0606],[246,-0.0622],[250,-0.0653],[267,-0.0825],[274,-0.0916],[276,-0.0938],[293,-0.1074],[296,-0.1092],[297,-0.1096],[298,-0.1095],[299,-0.1082],[300,-0.1067],[301,-0.1059],[302,-0.1061],[303,-0.1071],[304,-0.0995],[305,-0.0998],[306,-0.1058],[307,-0.1076],[308,-0.1076],[327,-0.1119],[345,-0.1121],[346,-0.1125],[347,-0.1137],[349,-0.1165],[350,-0.1171],[352,-0.1133],[353,-0.1122],[371,-0.1122],[373,-0.1129],[379,-0.1165],[381,-0.1171],[404,-0.1171],[405,-0.1178],[406,-0.1206],[407,-0.1251],[408,-0.1308],[410,-0.1435],[411,-0.1495],[412,-0.1546],[413,-0.1583],[414,-0.1601],[415,-0.1598],[416,-0.1582],[417,-0.1555],[418,-0.1519],[419,-0.1475],[420,-0.1427],[423,-0.127],[424,-0.122],[425,-0.1174],[426,-0.1135],[427,-0.1103],[428,-0.1081],[429,-0.1072],[487,-0.1072],[488,-0.1078],[489,-0.1086],[490,-0.1099],[491,-0.1114],[492,-0.1132],[494,-0.1176],[497,-0.1253],[503,-0.1418],[505,-0.1467],[507,-0.1508],[509,-0.1538],[510,-0.1549],[511,-0.1556],[576,-0.1559]]
       }
       CTRL_R_brow_inner {
        type=MovieSceneScriptingFloatChannel
        default=0
        thinned_from=577
        keys [[0,-0.0845],[41,-0.0845],[42,-0.0838],[43,-0.0796],[44,-0.0709],[45,-0.0593],[47,-0.0334],[48,-0.0221],[49,-0.0139],[50,-0.0082],[51,-0.0034],[52,0.0008],[53,0.0043],[54,0.0071],[55,0.0094],[56,0.0112],[57,0.0125],[58,0.0135],[59,0.0139],[94,0.0143],[100,0.0144],[101,0.0137],[102,0.0116],[103,0.0085],[104,0.0045],[105,-0.0002],[106,-0.0055],[107,-0.0115],[108,-0.0185],[109,-0.0262],[111,-0.042],[112,-0.0495],[113,-0.0562],[114,-0.0619],[115,-0.0664],[116,-0.07],[117,-0.0728],[118,-0.0752],[121,-0.0809],[122,-0.0824],[123,-0.0833],[124,-0.0836],[169,-0.0836],[171,-0.0826],[173,-0.0809],[175,-0.0784],[177,-0.0754],[178,-0.0737],[179,-0.0725],[180,-0.0722],[181,-0.0727],[182,-0.0738],[185,-0.0783],[186,-0.0796],[187,-0.0805],[188,-0.0807],[189,-0.0801],[190,-0.0785],[198,-0.0593],[200,-0.0551],[201,-0.0532],[202,-0.0516],[203,-0.0504],[204,-0.0498],[205,-0.0495],[241,-0.0494],[249,-0.0515],[275,-0.0626],[284,-0.064],[296,-0.0689],[297,-0.0692],[298,-0.0689],[299,-0.0675],[300,-0.0658],[301,-0.0649],[302,-0.0649],[309,-0.0657],[345,-0.0673],[346,-0.0677],[347,-0.0689],[348,-0.0704],[349,-0.0717],[350,-0.0723],[352,-0.0685],[353,-0.0674],[371,-0.0674],[373,-0.0681],[378,-0.0713],[380,-0.0721],[404,-0.0723],[405,-0.072],[406,-0.0709],[407,-0.0691],[408,-0.0667],[409,-0.0639],[413,-0.051],[415,-0.0453],[416,-0.0422],[418,-0.0353],[421,-0.0244],[423,-0.0176],[424,-0.0145],[425,-0.0118],[426,-0.0096],[427,-0.0078],[428,-0.0066],[429,-0.0061],[486,-0.0061],[487,-0.0063],[488,-0.0071],[489,-0.0085],[490,-0.0105],[491,-0.013],[492,-0.0159],[493,-0.0192],[494,-0.0228],[495,-0.0268],[497,-0.0352],[502,-0.0576],[503,-0.0618],[504,-0.0658],[505,-0.0696],[506,-0.0731],[507,-0.0762],[508,-0.0789],[509,-0.0811],[510,-0.0828],[511,-0.0839],[512,-0.0844],[576,-0.0844]]
       }
       CTRL_R_brow_outer {
        type=MovieSceneScriptingFloatChannel
        default=0
        thinned_from=577
        keys [[0,-0.069],[40,-0.069],[41,-0.0692],[42,-0.0687],[43,-0.0662],[44,-0.0625],[45,-0.0579],[47,-0.048],[48,-0.0435],[49,-0.0395],[50,-0.0361],[51,-0.0338],[52,-0.0322],[53,-0.0309],[54,-0.0299],[55,-0.0292],[57,-0.0284],[98,-0.0271],[101,-0.0269],[102,-0.0266],[103,-0.0258],[104,-0.0246],[105,-0.0232],[106,-0.0205],[108,-0.0138],[110,-0.0065],[111,-0.0031],[112,-0.0002],[113,0.0021],[114,0.0026],[156,0.0024],[239,0.0023],[241,0.0019],[243,0.001],[246,-0.0011],[249,-0.004],[255,-0.0114],[264,-0.0234],[266,-0.0259],[268,-0.0289],[274,-0.0392],[276,-0.042],[281,-0.0472],[284,-0.0496],[297,-0.0546],[327,-0.0619],[404,-0.0622],[405,-0.0629],[406,-0.0657],[407,-0.0702],[408,-0.0758],[410,-0.0885],[411,-0.0945],[412,-0.0996],[413,-0.1033],[414,-0.105],[415,-0.1047],[416,-0.1029],[417,-0.0998],[418,-0.0958],[419,-0.091],[420,-0.0856],[423,-0.068],[424,-0.0624],[425,-0.0572],[426,-0.0528],[427,-0.0492],[428,-0.0468],[429,-0.0457],[487,-0.0457],[489,-0.0464],[491,-0.0477],[493,-0.0495],[497,-0.0543],[503,-0.0622],[506,-0.0656],[508,-0.0673],[510,-0.0685],[512,-0.069],[576,-0.069]]
       }
       CTRL_R_eye_blink {
        type=MovieSceneScriptingFloatChannel
        default=0
        thinned_from=577
        keys [[0,0],[37,0],[38,0.1804],[39,0.5876],[40,1],[41,1],[42,0.9513],[43,0.8181],[44,0.3682],[45,0],[100,0],[101,0.03],[102,0.096],[103,0.4962],[104,1],[105,0.9867],[106,0.9575],[107,0.8181],[108,0.5012],[109,0.2478],[110,0.1042],[111,0.054],[112,0.0154],[113,0],[413,0],[414,0.4055],[415,0.892],[416,1],[417,0.9629],[418,0.8659],[419,0.7302],[420,0.577],[421,0.4275],[422,0.303],[423,0.2247],[424,0.1769],[425,0.131],[426,0.0891],[427,0.0531],[428,0.0249],[429,0.0066],[430,0],[440,0],[441,0.0586],[442,0.3813],[443,0.9028],[444,1],[445,0.9555],[446,0.8577],[447,0.4955],[448,0.0279],[449,0],[481,0],[482,0.0183],[483,0.0586],[484,0.3813],[485,0.9028],[486,1],[487,0.9555],[488,0.8577],[489,0.4955],[490,0.0279],[491,0],[521,0],[522,0.0183],[523,0.0586],[524,0.3813],[525,0.9029],[526,1],[527,0.9555],[528,0.8577],[529,0.4955],[530,0.0279],[531,0],[576,0]]
       }
       CTRL_R_eye_eyelidU.X {
        type=MovieSceneScriptingFloatChannel
        default=0
        thinned_from=577
        keys [[0,0.0292],[40,0.0292],[41,0.0418],[42,0.0646],[43,0.0851],[44,0.1048],[45,0.1228],[46,0.1382],[47,0.1503],[48,0.1582],[49,0.161],[98,0.161],[99,0.1602],[100,0.1583],[101,0.1559],[102,0.1418],[103,0.1267],[104,0.1121],[105,0.0986],[106,0.0868],[107,0.0761],[108,0.0657],[109,0.0559],[110,0.0472],[111,0.0399],[112,0.0342],[113,0.0305],[114,0.0292],[576,0.0292]]
       }
       CTRL_R_eye_eyelidU.Y {
        type=MovieSceneScriptingFloatChannel
        default=0
        thinned_from=577
        keys [[0,-0.0136],[41,-0.0136],[42,-0.0789],[43,-0.0638],[44,-0.0357],[45,-0.0206],[47,-0.021],[49,-0.022],[55,-0.0264],[57,-0.0274],[59,-0.0278],[98,-0.0278],[99,-0.0272],[100,-0.0255],[102,-0.0212],[103,-0.0195],[104,-0.0188],[105,-0.0789],[106,-0.0711],[107,-0.0535],[108,-0.035],[109,-0.0242],[110,-0.0203],[111,-0.0169],[112,-0.0145],[113,-0.0136],[204,-0.0136],[243,-0.0137],[245,-0.0139],[256,-0.0169],[276,-0.0249],[282,-0.0261],[289,-0.0263],[330,-0.0174],[334,-0.0157],[347,-0.0071],[350,-0.0057],[353,-0.0049],[399,-0.0048],[400,-0.005],[401,-0.0057],[402,-0.0068],[403,-0.0081],[406,-0.0132],[407,-0.0196],[408,-0.0328],[409,-0.0489],[410,-0.0644],[411,-0.0773],[412,-0.085],[413,-0.088],[414,-0.0894],[415,-0.0894],[416,-0.1482],[417,-0.1496],[418,-0.1456],[419,-0.1348],[420,-0.1193],[421,-0.1012],[422,-0.0824],[423,-0.065],[424,-0.0492],[425,-0.0343],[426,-0.0211],[427,-0.0106],[428,-0.0036],[429,-0.001],[435,-0.0008],[436,-0.0043],[437,-0.0108],[438,-0.0143],[439,-0.0136],[443,-0.0136],[444,-0.0716],[445,-0.0726],[446,-0.0533],[447,-0.0285],[448,-0.0183],[449,-0.0136],[485,-0.0136],[486,-0.0716],[487,-0.0726],[488,-0.0533],[489,-0.0285],[490,-0.0183],[491,-0.0136],[525,-0.0136],[526,-0.0716],[527,-0.0726],[528,-0.0533],[529,-0.0285],[530,-0.0183],[531,-0.0136],[576,-0.0136]]
       }
       CTRL_R_eye_eyelidD.X {type=MovieSceneScriptingFloatChannel, default=0, const [0,577]}
       CTRL_R_eye_eyelidD.Y {
        type=MovieSceneScriptingFloatChannel
        default=0
        thinned_from=577
        keys [[0,-0.0189],[41,-0.0189],[42,-0.1161],[43,-0.0909],[44,-0.0441],[45,-0.0189],[104,-0.0189],[105,-0.1161],[106,-0.1032],[107,-0.0743],[108,-0.0445],[109,-0.0287],[110,-0.0251],[111,-0.022],[112,-0.0197],[113,-0.0189],[171,-0.0189],[175,-0.0179],[180,-0.0156],[181,-0.0141],[182,-0.0108],[183,-0.0062],[184,-0.0007],[186,0.0111],[187,0.0164],[188,0.0208],[189,0.0238],[190,0.0249],[191,0.024],[192,0.0215],[193,0.0179],[194,0.0134],[195,0.0083],[197,-0.0022],[198,-0.0069],[199,-0.0108],[200,-0.0137],[201,-0.015],[207,-0.0179],[210,-0.0187],[218,-0.0188],[229,-0.0181],[258,-0.0141],[264,-0.0119],[278,-0.0036],[279,-0.0025],[280,-0.0009],[281,0.0014],[282,0.0043],[283,0.0076],[284,0.0113],[285,0.0153],[286,0.0196],[291,0.0428],[293,0.0517],[294,0.0557],[295,0.0595],[296,0.0628],[297,0.0657],[298,0.0681],[299,0.0698],[300,0.0709],[301,0.0713],[302,0.071],[303,0.0702],[304,0.0689],[305,0.0671],[306,0.0648],[307,0.0622],[308,0.0593],[309,0.056],[310,0.0525],[312,0.0449],[318,0.0198],[320,0.0119],[321,0.0081],[322,0.0046],[323,0.0014],[324,-0.0016],[325,-0.0042],[326,-0.0064],[327,-0.0082],[328,-0.0096],[329,-0.0104],[330,-0.0107],[333,-0.0107],[344,-0.0094],[350,-0.0091],[351,-0.0095],[352,-0.0103],[353,-0.0107],[357,-0.0107],[383,-0.0091],[390,-0.011],[406,-0.0177],[411,-0.0188],[415,-0.0189],[416,-0.1046],[417,-0.1152],[418,-0.1136],[419,-0.109],[420,-0.1022],[421,-0.0934],[422,-0.0834],[423,-0.0726],[424,-0.0615],[425,-0.0507],[426,-0.0406],[427,-0.0319],[428,-0.025],[429,-0.0205],[430,-0.0189],[443,-0.0189],[444,-0.1061],[445,-0.107],[446,-0.0783],[447,-0.0412],[448,-0.0259],[449,-0.0189],[485,-0.0189],[486,-0.1061],[487,-0.107],[488,-0.0783],[489,-0.0412],[490,-0.0259],[491,-0.0189],[525,-0.0189],[526,-0.1061],[527,-0.107],[528,-0.0783],[529,-0.0412],[530,-0.0259],[531,-0.0189],[576,-0.0189]]
       }
       CTRL_R_eye_squintInner {
        type=MovieSceneScriptingFloatChannel
        default=0
        thinned_from=577
        keys [[0,0],[105,0],[106,0.0086],[107,0.0241],[108,0.0335],[109,0.0337],[110,0.0281],[111,0.018],[112,0.0093],[113,0.0027],[114,0],[182,0],[183,0.0129],[184,0.0275],[185,0.0388],[186,0.0433],[187,0.0415],[188,0.037],[189,0.0312],[190,0.0186],[191,0.0066],[192,0.0021],[193,0],[245,0],[254,0.0006],[275,0.0037],[277,0.0044],[298,0.0187],[306,0.0226],[322,0.0273],[326,0.0279],[402,0.028],[403,0.0283],[404,0.029],[405,0.03],[407,0.0324],[408,0.0344],[409,0.0375],[410,0.0409],[411,0.0441],[412,0.0465],[413,0.0474],[414,0.0293],[415,0.0064],[416,0],[576,0]]
       }
       CTRL_R_eye_cheekRaise {
        type=MovieSceneScriptingFloatChannel
        default=0
        thinned_from=577
        keys [[0,0],[41,0],[42,0.0421],[43,0.0673],[44,0.05],[45,0.0178],[46,0],[104,0],[105,0.0421],[106,0.0388],[107,0.0328],[108,0.0258],[109,0.0184],[110,0.0115],[111,0.0056],[112,0.0015],[113,0],[114,0],[115,0.006],[116,0.0103],[117,0.0132],[118,0.015],[119,0.0159],[120,0.0163],[182,0.0163],[183,0.0115],[184,0.0062],[185,0.0018],[186,0],[240,0],[244,0.0013],[248,0.0037],[255,0.0097],[266,0.0206],[268,0.0232],[274,0.0323],[277,0.0361],[279,0.0381],[280,0.0389],[281,0.0405],[282,0.0439],[283,0.0487],[284,0.0548],[285,0.0616],[286,0.0691],[287,0.0773],[288,0.0861],[289,0.0953],[293,0.1331],[294,0.1421],[295,0.1505],[296,0.1583],[297,0.1653],[298,0.1714],[299,0.1759],[300,0.1788],[301,0.1798],[302,0.1792],[303,0.1777],[304,0.1751],[305,0.1718],[306,0.1676],[307,0.1627],[308,0.1572],[309,0.1511],[310,0.1446],[311,0.1376],[312,0.1302],[314,0.1149],[317,0.0913],[318,0.0836],[319,0.0762],[320,0.069],[321,0.0622],[322,0.0557],[323,0.0498],[324,0.0445],[325,0.0398],[326,0.0357],[327,0.0325],[328,0.0301],[329,0.0286],[330,0.0281],[404,0.0281],[405,0.0285],[406,0.03],[407,0.0324],[408,0.0354],[410,0.0417],[411,0.0445],[412,0.0465],[413,0.0474],[414,0.0364],[415,0.0064],[416,0.0319],[417,0.0643],[418,0.0666],[419,0.0634],[420,0.0583],[421,0.0517],[422,0.044],[424,0.0273],[425,0.0193],[426,0.0121],[427,0.0062],[428,0.002],[429,0.0001],[443,0.0001],[444,0.0319],[445,0.0647],[446,0.0567],[447,0.0254],[448,0.0014],[449,0],[485,0],[486,0.0319],[487,0.0647],[488,0.0567],[489,0.0254],[490,0.0014],[491,0],[525,0],[526,0.0319],[527,0.0647],[528,0.0567],[529,0.0254],[530,0.0014],[531,0],[576,0]]
       }
       CTRL_R_mouth_suckBlow {type=MovieSceneScriptingFloatChannel, default=0, const [0,577]}
       CTRL_C_eye.X {type=MovieSceneScriptingFloatChannel, default=0, thinned_from=577, keys [[0,0],[576,0]]}
       CTRL_C_eye.Y {type=MovieSceneScriptingFloatChannel, default=0, thinned_from=577, keys [[0,0],[287,0],[288,0.0543],[292,0.0543],[293,0],[576,0]]}
       CTRL_L_eye.X {
        type=MovieSceneScriptingFloatChannel
        default=0
        thinned_from=570
        keys [[0,0.4008],[37,0.4017],[38,0.3963],[40,0.3729],[41,0.3676],[47,0.9105],[50,0.9105],[51,0.9048],[52,0.8901],[53,0.8699],[54,0.848],[55,0.8279],[56,0.8133],[60,0.773],[62,0.7572],[63,0.7512],[64,0.7467],[84,0.7173],[91,0.7151],[92,0.7167],[93,0.7212],[94,0.728],[99,0.7732],[100,0.78],[101,0.7845],[102,0.7862],[105,0.2243],[106,0.2521],[107,0.2756],[108,0.2936],[109,0.3092],[110,0.3203],[113,0.3434],[114,0.3503],[115,0.3564],[116,0.3615],[117,0.3656],[127,0.3849],[132,0.3907],[161,0.4034],[164,0.4035],[165,0.4025],[166,0.3995],[167,0.3947],[168,0.3882],[169,0.3801],[170,0.3706],[171,0.3598],[172,0.3478],[173,0.3348],[174,0.3209],[175,0.3061],[177,0.2748],[182,0.1921],[183,0.1758],[184,0.1599],[185,0.1445],[186,0.1298],[187,0.1158],[188,0.1028],[189,0.0908],[190,0.08],[191,0.0705],[192,0.0624],[193,0.0559],[194,0.0511],[195,0.0481],[196,0.0471],[197,0.0478],[198,0.0496],[199,0.0527],[200,0.0567],[201,0.0616],[202,0.0674],[203,0.0739],[204,0.0811],[205,0.0887],[206,0.0968],[207,0.1052],[211,0.1398],[213,0.1564],[214,0.1641],[215,0.1714],[216,0.178],[217,0.184],[218,0.1891],[219,0.1934],[220,0.1966],[229,0.2199],[234,0.2305],[238,0.2369],[241,0.24],[244,0.2416],[262,0.238],[269,0.2312],[271,0.2304],[272,0.2306],[274,0.2325],[276,0.236],[278,0.2411],[280,0.2476],[282,0.2553],[285,0.2688],[290,0.2952],[302,0.3642],[306,0.3846],[310,0.4015],[313,0.4111],[316,0.4175],[332,0.4262],[365,0.4232],[374,0.4319],[395,0.4654],[402,0.4702],[404,0.4693],[405,0.4666],[406,0.4624],[407,0.4569],[409,0.4427],[415,0.3902],[416,0.3823],[417,0.3751],[418,0.369],[419,0.364],[420,0.3605],[426,0.345],[429,0.3395],[431,0.3373],[433,0.3365],[435,0.3367],[436,0.335],[437,0.3301],[438,0.3228],[439,0.3139],[442,0.2835],[443,0.2746],[444,0.2673],[445,0.2625],[446,0.2607],[447,0.2612],[448,0.2628],[449,0.2655],[450,0.269],[451,0.2735],[452,0.2788],[453,0.2848],[454,0.2916],[456,0.307],[458,0.3246],[461,0.3538],[470,0.4481],[473,0.4763],[475,0.4928],[477,0.507],[479,0.5183],[481,0.5263],[483,0.5306],[484,0.5311],[485,0.088],[486,0.0377],[487,-0.0095],[488,-0.0065],[489,0.0018],[490,0.0145],[491,0.0308],[492,0.0495],[493,0.0699],[494,0.0909],[495,0.1116],[496,0.1311],[497,0.1484],[498,0.1627],[499,0.1729],[500,0.1782],[506,0.1922],[509,0.1979],[511,0.2009],[513,0.203],[515,0.2041],[516,0.2043],[517,0.2004],[518,0.19],[519,0.1754],[521,0.142],[522,0.1274],[523,0.117],[524,0.1131],[525,0.2195],[526,0.4229],[527,0.5495],[528,0.5697],[529,0.5608],[530,0.5381],[531,0.508],[532,0.4766],[533,0.4503],[534,0.4353],[536,0.4212],[537,0.4151],[538,0.4102],[539,0.4068],[540,0.4052],[576,0.4007]]
       }
       CTRL_L_eye.Y {
        type=MovieSceneScriptingFloatChannel
        default=0
        thinned_from=570
        keys [[0,0.0233],[1,0.0231],[2,0.0225],[3,0.0217],[4,0.0205],[6,0.0177],[11,0.0093],[13,0.0065],[14,0.0054],[15,0.0045],[16,0.004],[17,0.0038],[18,0.004],[19,0.0048],[20,0.006],[21,0.0076],[22,0.0095],[23,0.0117],[25,0.0166],[29,0.0275],[31,0.0325],[32,0.0346],[33,0.0365],[34,0.0381],[35,0.0393],[36,0.0401],[37,0.0404],[38,0.0292],[40,-0.0199],[41,-0.031],[47,0.0333],[48,0.0261],[49,0.0185],[51,0.0028],[52,-0.0047],[53,-0.0116],[54,-0.0177],[55,-0.0227],[56,-0.0264],[57,-0.0284],[63,-0.0348],[66,-0.0375],[69,-0.0394],[71,-0.0401],[72,-0.0403],[74,-0.0403],[76,-0.0398],[78,-0.0387],[81,-0.0364],[92,-0.0252],[95,-0.0227],[97,-0.0215],[99,-0.0206],[100,-0.0204],[101,-0.0204],[102,-0.0261],[105,-0.0373],[106,-0.039],[107,-0.0402],[109,-0.0415],[110,-0.0417],[126,-0.0385],[135,-0.0384],[139,-0.0379],[148,-0.0373],[149,-0.0372],[150,-0.0367],[151,-0.036],[153,-0.034],[155,-0.0313],[163,-0.0181],[165,-0.0153],[167,-0.0129],[168,-0.012],[169,-0.0114],[170,-0.011],[171,-0.0108],[172,-0.0116],[173,-0.0137],[174,-0.0172],[175,-0.0218],[176,-0.0274],[177,-0.0338],[178,-0.041],[179,-0.0488],[180,-0.057],[184,-0.0915],[185,-0.0999],[186,-0.1079],[187,-0.1154],[192,-0.1509],[195,-0.1713],[198,-0.1905],[200,-0.2023],[202,-0.2133],[204,-0.2233],[206,-0.2321],[208,-0.2396],[210,-0.2457],[212,-0.2502],[214,-0.253],[216,-0.254],[217,-0.2534],[218,-0.2519],[219,-0.2496],[221,-0.2431],[225,-0.2276],[226,-0.2241],[227,-0.2211],[228,-0.2188],[229,-0.2173],[230,-0.2167],[232,-0.2176],[234,-0.2199],[236,-0.2236],[238,-0.2283],[241,-0.2372],[254,-0.2835],[257,-0.2924],[260,-0.2991],[262,-0.3022],[264,-0.3038],[265,-0.304],[266,-0.3032],[267,-0.3011],[268,-0.2978],[269,-0.2933],[270,-0.2878],[271,-0.2813],[272,-0.2741],[273,-0.2662],[275,-0.2487],[278,-0.2203],[279,-0.2091],[280,-0.1953],[281,-0.1795],[282,-0.1623],[283,-0.1441],[285,-0.1072],[286,-0.0896],[287,-0.0734],[288,-0.0591],[289,-0.0472],[290,-0.0384],[291,-0.0332],[297,-0.0141],[300,-0.0052],[303,0.0029],[305,0.0078],[307,0.0122],[309,0.016],[311,0.0191],[313,0.0214],[314,0.0223],[315,0.0229],[316,0.0234],[321,0.0237],[328,0.0228],[331,0.0227],[334,0.0236],[344,0.0291],[347,0.0302],[349,0.0304],[350,0.0302],[351,0.0295],[352,0.0282],[353,0.0266],[354,0.0247],[355,0.0225],[357,0.0175],[361,0.0066],[363,0.0016],[364,-0.0006],[365,-0.0026],[366,-0.0042],[367,-0.0054],[368,-0.0061],[369,-0.0064],[370,-0.0064],[372,-0.0056],[374,-0.0043],[376,-0.0024],[378,0],[381,0.0044],[384,0.0097],[392,0.0258],[393,0.0283],[394,0.0317],[395,0.0359],[396,0.0408],[397,0.0462],[398,0.0522],[399,0.0586],[400,0.0654],[402,0.0796],[405,0.1013],[406,0.1083],[407,0.115],[408,0.1214],[409,0.1274],[410,0.1328],[411,0.1376],[412,0.1418],[413,0.1452],[414,0.1477],[415,0.1493],[416,0.1498],[417,0.1303],[418,0.0844],[419,0.0312],[420,-0.0101],[421,-0.039],[422,-0.0666],[423,-0.0915],[424,-0.1125],[425,-0.1283],[426,-0.1374],[429,-0.1534],[431,-0.1632],[432,-0.1677],[433,-0.1718],[434,-0.1754],[435,-0.1785],[436,-0.181],[437,-0.1829],[438,-0.1841],[439,-0.1845],[440,-0.1842],[442,-0.182],[445,-0.1779],[446,-0.177],[447,-0.1767],[448,-0.1786],[449,-0.1842],[450,-0.1931],[451,-0.2049],[452,-0.2191],[453,-0.2355],[454,-0.2535],[455,-0.273],[457,-0.3143],[459,-0.3564],[460,-0.3768],[461,-0.3963],[462,-0.4143],[463,-0.4307],[464,-0.4449],[465,-0.4567],[466,-0.4656],[467,-0.4712],[468,-0.4731],[469,-0.472],[470,-0.4689],[471,-0.4641],[472,-0.4579],[474,-0.4424],[478,-0.4067],[479,-0.3985],[480,-0.3912],[481,-0.385],[482,-0.3802],[483,-0.3771],[484,-0.376],[485,-0.4035],[486,-0.4062],[487,-0.4028],[488,-0.3934],[489,-0.3793],[490,-0.3618],[493,-0.3029],[494,-0.2855],[495,-0.2714],[496,-0.2619],[499,-0.2423],[501,-0.2304],[502,-0.2251],[503,-0.2204],[504,-0.2164],[505,-0.213],[506,-0.2105],[507,-0.209],[508,-0.2084],[509,-0.2092],[510,-0.2113],[511,-0.2146],[512,-0.2188],[513,-0.2237],[517,-0.2456],[518,-0.2505],[519,-0.2547],[520,-0.258],[521,-0.2601],[522,-0.2609],[523,-0.2532],[524,-0.2343],[525,-0.2104],[526,-0.1878],[527,-0.1727],[530,-0.1449],[531,-0.1335],[532,-0.1185],[533,-0.1007],[534,-0.0811],[536,-0.0397],[537,-0.0197],[538,-0.0014],[539,0.0144],[540,0.0268],[541,0.0349],[542,0.0378],[543,0.0372],[544,0.0356],[545,0.0334],[546,0.0306],[548,0.0245],[549,0.0217],[550,0.0194],[551,0.0179],[552,0.0173],[553,0.0175],[554,0.0179],[559,0.0217],[560,0.0224],[561,0.0228],[576,0.023]]
       }
       CTRL_R_eye.X {
        type=MovieSceneScriptingFloatChannel
        default=0
        thinned_from=570
        keys [[0,0.4],[36,0.401],[37,0.394],[38,0.3807],[40,0.3696],[41,0.3672],[47,0.9105],[50,0.9105],[51,0.9047],[52,0.8898],[53,0.8694],[54,0.8473],[55,0.827],[56,0.8124],[60,0.7722],[62,0.7565],[63,0.7505],[64,0.7461],[84,0.7168],[91,0.7145],[92,0.7161],[93,0.7207],[94,0.7274],[99,0.7724],[100,0.7792],[101,0.7837],[102,0.7854],[105,0.143],[106,0.1697],[108,0.2316],[109,0.2592],[110,0.2796],[112,0.31],[113,0.3241],[114,0.337],[115,0.3482],[116,0.3573],[117,0.3639],[118,0.3675],[128,0.3854],[161,0.4024],[164,0.4025],[165,0.4015],[166,0.3985],[167,0.3937],[168,0.3872],[169,0.3792],[170,0.3697],[171,0.3589],[172,0.347],[173,0.334],[174,0.3201],[175,0.3054],[177,0.2742],[182,0.1917],[183,0.1754],[184,0.1595],[185,0.1442],[186,0.1295],[187,0.1156],[188,0.1026],[189,0.0907],[190,0.0799],[191,0.0704],[192,0.0624],[193,0.0559],[194,0.0511],[195,0.0481],[196,0.0471],[197,0.0477],[198,0.0496],[199,0.0526],[200,0.0566],[201,0.0616],[202,0.0674],[203,0.0739],[204,0.081],[205,0.0886],[206,0.0966],[207,0.105],[211,0.1396],[213,0.1561],[214,0.1639],[215,0.1711],[216,0.1778],[217,0.1837],[218,0.1888],[219,0.1931],[220,0.1963],[229,0.2196],[234,0.2303],[238,0.2366],[241,0.2398],[244,0.2413],[264,0.238],[265,0.2375],[269,0.2315],[270,0.2304],[271,0.23],[272,0.2302],[274,0.232],[276,0.2356],[278,0.2407],[280,0.2472],[282,0.2548],[285,0.2683],[290,0.2946],[302,0.3634],[306,0.3837],[310,0.4005],[313,0.4101],[316,0.4165],[332,0.4252],[361,0.4219],[371,0.4303],[395,0.4648],[402,0.4689],[404,0.4685],[406,0.4652],[408,0.4591],[410,0.4506],[413,0.4345],[423,0.3704],[425,0.3593],[427,0.3498],[429,0.3424],[430,0.3396],[431,0.3376],[432,0.3364],[435,0.3362],[436,0.3344],[437,0.3295],[438,0.3223],[439,0.3134],[442,0.2831],[443,0.2741],[444,0.2669],[445,0.262],[446,0.2603],[447,0.2608],[448,0.2624],[449,0.265],[450,0.2686],[451,0.273],[452,0.2783],[453,0.2843],[454,0.291],[456,0.3064],[458,0.3239],[461,0.353],[470,0.4468],[473,0.4749],[475,0.4914],[477,0.5055],[479,0.5168],[481,0.5247],[483,0.529],[484,0.5295],[485,0.3102],[486,0.0388],[487,-0.0081],[488,-0.0056],[489,0.0016],[490,0.0127],[491,0.0269],[492,0.0435],[493,0.0618],[495,0.1004],[496,0.1192],[497,0.1368],[498,0.1523],[499,0.165],[500,0.1742],[501,0.1792],[506,0.1918],[509,0.1981],[511,0.2014],[513,0.2038],[515,0.2051],[516,0.2053],[517,0.2014],[518,0.1911],[519,0.1765],[521,0.1431],[522,0.1286],[523,0.1183],[524,0.1144],[525,0.2201],[526,0.4221],[527,0.5479],[528,0.5679],[529,0.559],[530,0.5365],[531,0.5065],[532,0.4753],[533,0.4492],[534,0.4343],[536,0.4203],[537,0.4142],[538,0.4093],[539,0.4059],[540,0.4044],[576,0.3999]]
       }
       CTRL_R_eye.Y {
        type=MovieSceneScriptingFloatChannel
        default=0
        thinned_from=570
        keys [[0,0.0257],[1,0.0255],[2,0.0249],[3,0.0241],[4,0.0229],[6,0.0201],[11,0.0117],[13,0.0089],[14,0.0078],[15,0.0069],[16,0.0064],[17,0.0062],[18,0.0065],[19,0.0073],[20,0.0086],[21,0.0103],[22,0.0124],[23,0.0147],[25,0.0201],[29,0.0313],[30,0.0339],[31,0.0363],[32,0.0383],[33,0.0401],[34,0.0414],[35,0.0422],[36,0.0425],[37,0.028],[38,-0.0005],[39,-0.0132],[40,-0.0245],[41,-0.0297],[47,0.0348],[48,0.0276],[49,0.02],[51,0.0043],[52,-0.0032],[53,-0.0101],[54,-0.0162],[55,-0.0212],[56,-0.0248],[57,-0.0268],[63,-0.0333],[66,-0.0359],[69,-0.0378],[71,-0.0386],[72,-0.0388],[74,-0.0388],[76,-0.0382],[78,-0.0372],[81,-0.0349],[92,-0.0237],[95,-0.0212],[97,-0.02],[99,-0.0192],[100,-0.0189],[101,-0.0189],[102,-0.0195],[105,-0.0273],[107,-0.0343],[108,-0.0376],[109,-0.0402],[110,-0.042],[111,-0.0427],[112,-0.0427],[116,-0.0415],[123,-0.0383],[126,-0.0376],[147,-0.0365],[148,-0.0363],[149,-0.0359],[150,-0.0353],[152,-0.0334],[154,-0.0309],[163,-0.0167],[165,-0.014],[167,-0.0119],[168,-0.0111],[169,-0.0104],[170,-0.0101],[171,-0.0099],[172,-0.0107],[173,-0.0129],[174,-0.0163],[175,-0.0209],[176,-0.0265],[177,-0.033],[178,-0.0402],[179,-0.048],[180,-0.0562],[184,-0.0907],[185,-0.099],[186,-0.1069],[187,-0.1143],[192,-0.1494],[195,-0.1696],[198,-0.1886],[200,-0.2003],[202,-0.2112],[204,-0.2211],[206,-0.2298],[208,-0.2372],[210,-0.2432],[212,-0.2477],[214,-0.2505],[216,-0.2514],[217,-0.2509],[218,-0.2495],[219,-0.2474],[221,-0.2415],[225,-0.2274],[227,-0.2215],[228,-0.2193],[229,-0.2179],[230,-0.2174],[232,-0.2183],[234,-0.2206],[236,-0.2243],[238,-0.2292],[241,-0.2381],[254,-0.2851],[257,-0.294],[260,-0.3009],[262,-0.3039],[264,-0.3055],[265,-0.3057],[266,-0.305],[267,-0.3029],[268,-0.2994],[269,-0.2948],[270,-0.2891],[271,-0.2825],[272,-0.2751],[273,-0.267],[275,-0.2491],[278,-0.2199],[279,-0.2085],[280,-0.1945],[281,-0.1785],[282,-0.1611],[283,-0.1427],[285,-0.1056],[286,-0.088],[287,-0.0717],[288,-0.0573],[289,-0.0454],[290,-0.0365],[291,-0.0313],[297,-0.0121],[300,-0.0032],[302,0.0024],[304,0.0075],[306,0.0122],[308,0.0163],[310,0.0198],[312,0.0225],[313,0.0236],[314,0.0245],[315,0.0251],[316,0.0255],[321,0.0258],[328,0.025],[331,0.0251],[335,0.0265],[345,0.0317],[348,0.0326],[350,0.0328],[351,0.0325],[352,0.0317],[353,0.0303],[354,0.0285],[355,0.0264],[356,0.024],[358,0.0185],[362,0.0068],[363,0.0041],[364,0.0017],[365,-0.0004],[366,-0.0022],[367,-0.0036],[368,-0.0044],[369,-0.0047],[370,-0.0047],[372,-0.0039],[374,-0.0026],[376,-0.0007],[378,0.0018],[381,0.0062],[384,0.0115],[392,0.0278],[393,0.0304],[394,0.0339],[395,0.0383],[396,0.0433],[397,0.049],[398,0.0552],[399,0.0618],[400,0.0688],[405,0.1054],[406,0.1124],[407,0.119],[408,0.1253],[409,0.131],[410,0.1361],[411,0.1405],[412,0.144],[413,0.1467],[414,0.1484],[415,0.149],[416,0.1449],[417,0.1336],[418,0.1164],[419,0.0944],[420,0.069],[421,0.0414],[422,0.0129],[423,-0.0152],[424,-0.0418],[425,-0.0654],[426,-0.0849],[427,-0.0991],[429,-0.1209],[430,-0.1314],[431,-0.1415],[432,-0.1509],[433,-0.1596],[434,-0.1674],[435,-0.1741],[436,-0.1795],[437,-0.1836],[438,-0.1862],[439,-0.1871],[440,-0.1867],[441,-0.1858],[445,-0.1799],[446,-0.1789],[447,-0.1785],[448,-0.1805],[449,-0.186],[450,-0.1947],[451,-0.2063],[452,-0.2204],[453,-0.2366],[454,-0.2544],[455,-0.2736],[457,-0.3144],[459,-0.356],[460,-0.3761],[461,-0.3953],[462,-0.4132],[463,-0.4293],[464,-0.4434],[465,-0.455],[466,-0.4637],[467,-0.4693],[468,-0.4712],[469,-0.4701],[470,-0.467],[471,-0.4622],[472,-0.456],[474,-0.4404],[478,-0.4046],[479,-0.3963],[480,-0.3889],[481,-0.3827],[482,-0.3779],[483,-0.3748],[484,-0.3737],[486,-0.4039],[487,-0.4011],[488,-0.3935],[489,-0.3819],[490,-0.3674],[491,-0.3508],[493,-0.3151],[494,-0.2979],[495,-0.2824],[496,-0.2694],[497,-0.2601],[500,-0.2388],[501,-0.2323],[502,-0.2262],[503,-0.2208],[504,-0.2161],[505,-0.2122],[506,-0.2092],[507,-0.2074],[508,-0.2068],[509,-0.2075],[510,-0.2097],[511,-0.2129],[512,-0.2171],[513,-0.222],[517,-0.2439],[518,-0.2488],[519,-0.253],[520,-0.2563],[521,-0.2584],[522,-0.2592],[523,-0.2515],[524,-0.2327],[525,-0.2089],[526,-0.1863],[527,-0.1712],[530,-0.1433],[531,-0.1319],[532,-0.1168],[533,-0.0989],[534,-0.0792],[536,-0.0376],[537,-0.0176],[538,0.0009],[539,0.0167],[540,0.0291],[541,0.0372],[542,0.0401],[543,0.0396],[544,0.038],[545,0.0357],[546,0.033],[548,0.0269],[549,0.0241],[550,0.0218],[551,0.0203],[552,0.0197],[553,0.0199],[554,0.0203],[559,0.0241],[560,0.0248],[561,0.0252],[576,0.0253]]
       }
       eyesConvergence {type=MovieSceneScriptingFloatChannel, default=0, const [0,577]}
       eyelidsReduce {type=MovieSceneScriptingFloatChannel, default=0, const [0,577]}
       lookAtSwitch {type=MovieSceneScriptingFloatChannel, default=0, const [0,577]}
       CTRL_C_tongue.Location.X {type=MovieSceneScriptingFloatChannel, default=0, const [0,577]}
       CTRL_C_tongue.Location.Y {type=MovieSceneScriptingFloatChannel, default=0, const [0,577]}
       CTRL_C_tongue.Location.Z {type=MovieSceneScriptingFloatChannel, default=0, const [0,577]}
       wideNarrow {type=MovieSceneScriptingFloatChannel, default=0, const [0,577]}
       rollUpDown {type=MovieSceneScriptingFloatChannel, default=0, const [0,577]}
       rollLeftRight {type=MovieSceneScriptingFloatChannel, default=0, const [0,577]}
       CTRL_C_teethU.Location.X {type=MovieSceneScriptingFloatChannel, default=0, const [0,577]}
       CTRL_C_teethU.Location.Y {type=MovieSceneScriptingFloatChannel, default=0, const [0,577]}
       CTRL_C_teethU.Location.Z {type=MovieSceneScriptingFloatChannel, default=0, const [0,577]}
       rotateFwdBack {type=MovieSceneScriptingFloatChannel, default=0, const [0,577]}
       CTRL_C_teethD.Location.X {type=MovieSceneScriptingFloatChannel, default=0, const [0,577]}
       CTRL_C_teethD.Location.Y {type=MovieSceneScriptingFloatChannel, default=0, const [0,577]}
       CTRL_C_teethD.Location.Z {type=MovieSceneScriptingFloatChannel, default=0, const [0,577]}
       rotateFwdBack_2 {type=MovieSceneScriptingFloatChannel, default=0, const [0,577]}
       Weight {type=MovieSceneScriptingFloatChannel, default=1}
      ]
     }
     {
      type=MovieSceneControlRigParameterSection
      channels [
       CTRL_L_eye_blink {type=MovieSceneScriptingFloatChannel, default=0, keys [[305,0],[306,0.1117],[308,1],[310,1],[315,0]]}
       CTRL_L_eye_eyelidU.X {type=MovieSceneScriptingFloatChannel, default=0, const [0,4]}
       CTRL_L_eye_eyelidU.Y {type=MovieSceneScriptingFloatChannel, default=0, keys [[305,0],[308,-0.0125],[310,-0.3253],[315,0]]}
       CTRL_L_eye_eyelidD.X {type=MovieSceneScriptingFloatChannel, default=0, thinned_from=4, keys [[305,0],[315,0]]}
       CTRL_L_eye_eyelidD.Y {type=MovieSceneScriptingFloatChannel, default=0, keys [[305,0],[308,-0.0105],[310,-0.2719],[315,0]]}
       CTRL_L_brow.X {type=MovieSceneScriptingFloatChannel, default=0, const [0,4]}
       CTRL_L_brow.Y {type=MovieSceneScriptingFloatChannel, default=0, const [0,4]}
       CTRL_L_brow_inner {type=MovieSceneScriptingFloatChannel, default=0, const [0,4]}
       CTRL_L_brow_outer {type=MovieSceneScriptingFloatChannel, default=0, const [0,4]}
       CTRL_L_eye_squintInner {type=MovieSceneScriptingFloatChannel, default=0, const [0,4]}
       CTRL_R_brow.X {type=MovieSceneScriptingFloatChannel, default=0, const [0,4]}
       CTRL_R_brow.Y {type=MovieSceneScriptingFloatChannel, default=0, const [0,4]}
       CTRL_R_brow_inner {type=MovieSceneScriptingFloatChannel, default=0, const [0,4]}
       CTRL_R_brow_outer {type=MovieSceneScriptingFloatChannel, default=0, const [0,4]}
       CTRL_R_eye_blink {type=MovieSceneScriptingFloatChannel, default=0, keys [[305,0],[306,0.1117],[308,1],[310,1],[315,0]]}
       CTRL_R_eye_eyelidU.X {type=MovieSceneScriptingFloatChannel, default=0, const [0,4]}
       CTRL_R_eye_eyelidU.Y {type=MovieSceneScriptingFloatChannel, default=0, keys [[305,0],[308,-0.0125],[310,-0.3253],[315,0]]}
       CTRL_R_eye_eyelidD.X {type=MovieSceneScriptingFloatChannel, default=0, thinned_from=4, keys [[305,0],[315,0]]}
       CTRL_R_eye_eyelidD.Y {type=MovieSceneScriptingFloatChannel, default=0, keys [[305,0],[308,-0.0105],[310,-0.2719],[315,0]]}
       CTRL_R_eye_squintInner {type=MovieSceneScriptingFloatChannel, default=0, const [0,4]}
       CTRL_C_eye.X {
        type=MovieSceneScriptingFloatChannel
        default=0
        thinned_from=18
        keys [[169,0],[276,0],[285,0.0753],[298,0.0753],[299,0.106],[319,0.106],[320,0],[386,0],[387,0.106],[407,0.106],[408,0],[422,0]]
       }
       CTRL_C_eye.Y {
        type=MovieSceneScriptingFloatChannel
        default=0
        thinned_from=20
        keys [[169,0],[283,0],[284,-0.06],[349,-0.06],[350,0],[375,0],[376,-0.06],[421,-0.06],[422,0]]
       }
       CTRL_L_eye.X {type=MovieSceneScriptingFloatChannel, default=0, const [0,4]}
       CTRL_L_eye.Y {type=MovieSceneScriptingFloatChannel, default=0, const [0,4]}
       CTRL_R_eye.X {type=MovieSceneScriptingFloatChannel, default=0, const [0,4]}
       CTRL_R_eye.Y {type=MovieSceneScriptingFloatChannel, default=0, const [0,4]}
       eyesConvergence {type=MovieSceneScriptingFloatChannel, default=0, const [0,18]}
       eyelidsReduce {type=MovieSceneScriptingFloatChannel, default=0, const [0,18]}
       Weight {type=MovieSceneScriptingFloatChannel, default=1}
      ]
     }
    ]
   }
  ]
 }]
display_rate {denominator=1, numerator=30}
tick_resolution {denominator=1, numerator=24000}
tracks []
